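# Translucent, unlit, two-sided Niagara mesh particle material for the bowl mesh in the OboeThorn back bling effect.
# Scrolling-noise-distorted UVs form a sharpened V gradient whose hottest core boosts particle-color emissive.
# Opacity combines that gradient, particle alpha and a view-based fade using both a local axis and vertex normals, softening edge-on views.
# The graph also holds a toon-fire curve atlas color branch with hue shift and desaturation.
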Material /BRCosmetics/Accessories/FORT_Backpacks/Backpack_M_MED_OboeThorn/FX/M_FX_OboeThorn_ForBowlMesh
graph {
 properties {
  bUsedWithNiagaraMeshParticles=true
  blendMode=BLEND_Translucent
  pixelDepthOffsetMode=PDOM_Legacy
  shadingModel=MSM_Unlit
  twoSided=true
  usedShadingModels=Unlit
 }
 nodes [
  n1 = Constant() {r=2}
  n2 = Power(Base=NamedRerouteUsage(), Exp=Constant() {r=30})
  MaterialFunctionCall(Emissive=Multiply(A=ParticleColor().RGB, B=LinearInterpolate(B=@n1, Alpha=@n2) {constA=1})) {materialFunction=&MF_FX_EmissiveAdjustments}
  n3 = ScalarParameter() {parameterName=Desaturation}
  n4 = ScalarParameter() {group=Color, parameterName=Color Smooth Step Min, sliderMax=1, sliderMin=-1}
  n5 = ScalarParameter() {defaultValue=2, group=Color, parameterName=Color Smooth Step Max, sliderMax=5}
  n6 = ParticleColor()
  n7 = MaterialFunctionCall(Float2=Add(A=TextureCoordinate(), B=NamedRerouteUsage())) {materialFunction=&BreakOutFloat2Components}
  n8 = Multiply(A=@n6.A, B=Power(Base=@n7.G, Exp=Constant() {r=10}))
  n9 = Power(Base=Saturate(None=@n8), Exp=ScalarParameter() {defaultValue=1, group=Color, parameterName=Color Power, sliderMax=5}) {constExponent=1}
  n10 = CurveAtlasRowParameter(CurveTime=LinearInterpolate(B=SmoothStep(Min=@n4, Max=@n5, Value=@n9) {constMax=2}) {constA=1, constAlpha=1}) {atlas=/Game/Athena/LinearColorCurves/VFX_CurveAtlas_Toon, curve=/Game/Athena/LinearColorCurves/Curve_FireToon02, parameterName=Curve_Color}
  MaterialFunctionCall(Hue Shift Percentage=@n3, Texture=@n10) {materialFunction=&HueShift}
  Multiply(A=Desaturation(None=@n10, Fraction=@n3), B=@n6.RGB)
  n11 = VectorParameter() {parameterName=ViewBasedFadeLocalAxis, defaultValue=<rgba 0 0 1 1>}
  n12 = CameraVectorWS()
  n13 = Abs(None=DotProduct(A=Transform(None=@n11.RGB) {transformSourceType=TRANSFORMSOURCE_Local}, B=@n12))
  n14 = Abs(None=DotProduct(A=VertexNormalWS(), B=@n12))
  n15 = SmoothStep(
   Min=ScalarParameter() {parameterName=ViewBasedFadeMin, sliderMax=2, sliderMin=-2}
   Max=ScalarParameter() {defaultValue=1, parameterName=ViewBasedFadeMax, sliderMax=2, sliderMin=-2}
   Value=Power(Base=Min(A=@n13, B=@n14), Exp=ScalarParameter() {defaultValue=8, parameterName=ViewBasedFadeSensitivity, sliderMax=16})
  )
  Multiply(A=ScalarParameter() {defaultValue=1, parameterName=Overall Opacity, sliderMax=1}, B=Multiply(A=Clamp(None=@n15), B=@n8) {constA=1}) {constA=1, constB=-1}
  NamedRerouteDeclaration(None=@n7.G) {name=InnerCore, nodeColor=<rgba 0.8984 0 1 1>}
  n16 = VectorParameter() {parameterName=UVDistortTex_UVControl, defaultValue=<rgba 1 1 0 0.25>}
  n17 = MaterialFunctionCall!(UVInput, TimeInput, Control_RGB=@n16.RGB, Control_A=@n16.RGB, ScaleUVByCenter) {materialFunction=/Game/Effects/Niagara/MasterMaterials/Functions/Common/FXCommon_QuickUVControl_V2}
  NamedRerouteDeclaration(None=Multiply(
    A=TextureSample(UVs=@n17.UV) {samplerType=SAMPLERTYPE_Grayscale, texture=&LowResBlurredNoise}.R
    B=ScalarParameter() {defaultValue=0.1, parameterName=UV Distortion}
   )) {name=UVDistortion, nodeColor=<rgba 0.5703 1 0 1>}
 ]
}
