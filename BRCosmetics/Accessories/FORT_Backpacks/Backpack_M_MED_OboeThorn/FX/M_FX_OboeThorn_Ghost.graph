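# Translucent, unlit, two-sided Niagara sprite and mesh material that draws a ghost for the OboeThorn backpack.
# GhostTex supplies opacity, emissive and detail channels through sine-wobbled UVs, over teal colors blended by a noise-distorted radial gradient.
# Opacity combines depth fade, optional bottom and view-angle fades, and particle alpha.
# Vertex offset math twists around a local axis with a time-based sine scaled by a dynamic parameter, plus a camera-ward push.
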Material /BRCosmetics/Accessories/FORT_Backpacks/Backpack_M_MED_OboeThorn/FX/M_FX_OboeThorn_Ghost
graph {
 properties {
  bUsedWithNiagaraMeshParticles=true
  bUsedWithNiagaraSprites=true
  blendMode=BLEND_Translucent
  pixelDepthOffsetMode=PDOM_Legacy
  shadingModel=MSM_Unlit
  twoSided=true
  usedShadingModels=Unlit
 }
 nodes [
  n1 = Constant3Vector() {constant=<rgba 0 0 1 0>}
  n2 = Constant() {r=0.7}
  n3 = Multiply(A=ComponentMask(None=TextureCoordinate()) {g=true}, B=ScalarParameter() {defaultValue=0.5, parameterName=WaveAmount}) {constB=0.5}
  n4 = Time()
  n5 = Multiply(A=Sine(None=Add(A=@n3, B=@n4)), B=Multiply(A=Constant() {r=0.005}, B=DynamicParameter() {paramNames [FaceRotation,Twist,Time,Param4]}.Twist)) {constB=0.25}
  n6 = RotateAboutAxis(
   NormalizedRotationAxis=Normalize(VectorInput=Transform(None=@n1) {transformSourceType=TRANSFORMSOURCE_Local})
   RotationAngle=Multiply(A=Multiply(A=@n2, B=@n5))
   PivotPoint=ParticlePositionWS()
   Position=WorldPosition().XYZ
  )
  n7 = Subtract(A=CameraPositionWS(), B=Add(A=WorldPosition().XYZ, B=@n6))
  Add(A=Multiply(A=Normalize(VectorInput=@n7), B=ScalarParameter() {defaultValue=5, parameterName=VS Camera Offset}), B=@n6)
  n8 = Multiply(
   A=TextureCoordinate()
   B=MaterialFunctionCall(
    X=ScalarParameter() {defaultValue=1, parameterName=ColorDistortionUVX}
    Y=ScalarParameter() {defaultValue=0.5, parameterName=ColorDistortionUVY}
   ) {materialFunction=&MakeFloat2}.Result
  )
  n9 = AppendVector(A=Constant(), B=ScalarParameter() {defaultValue=-0.35, parameterName=ColorDistortionSpeed})
  n10 = ComponentMask(None=TextureSample(UVs=Panner(Coordinate=@n8, Speed=@n9) {speedY=0.2}) {samplerType=SAMPLERTYPE_Normal, texture=/Game/Accessories/FORT_Backpacks/Backpack_M_MED_EctoCat/FX/CloudNoise_T_N}.RGB) {g=true, r=true}
  n11 = Constant() {r=0.1}
  n12 = MaterialFunctionCall(UVs=Add(A=TextureCoordinate(), B=Multiply(A=@n10, B=@n11) {constB=0.3}), Radius=Constant() {r=0.5}, Density=Constant() {r=1}) {materialFunction=&RadialGradientExponential}
  n13 = TextureSample(
   UVs=NamedRerouteUsage()
   Tex=TextureObjectParameter() {parameterName=GhostTex, texture=&/T_FX_SmallGhost}
  )
  n14 = LinearInterpolate(
   A=LinearInterpolate(
    A=VectorParameter() {defaultValue=<rgba 0.1252 0.4487 0.5217 1>}.RGB
    B=VectorParameter() {parameterName=Param_1, defaultValue=<rgba 0.0984 0.4748 0.5391 1>}.RGB
    Alpha=Multiply(A=@n12.RadialGradientExponential)
   )
   B=VectorParameter() {parameterName=EmissiveColor}.RGB
   Alpha=LinearInterpolate(B=@n13.G, Alpha=ScalarParameter() {parameterName=UseEmissive?})
  )
  n15 = LinearInterpolate(A=ScalarParameter() {defaultValue=0.1, parameterName=DetailMultiplier}, Alpha=@n13.B)
  MaterialFunctionCall(Emissive=Multiply(A=Multiply(A=@n14, B=@n15), B=ParticleColor().RGB)) {materialFunction=&MF_FX_EmissiveAdjustments}
  n16 = ScalarParameter() {defaultValue=5, parameterName=FadeDistance}
  n17 = TextureSample(
   UVs=NamedRerouteUsage()
   Tex=TextureObjectParameter() {parameterName=GhostTex, texture=&T_FX_SmallGhost}
  ) {texture=&/T_FX_ReactiveGhost}
  n18 = TextureCoordinate()
  n19 = Power(
   Base=OneMinus(None=MaterialFunctionCall(Float2=@n18) {materialFunction=&BreakOutFloat2Components}.G)
   Exp=ScalarParameter() {defaultValue=3, parameterName=BottomFadePower}
  )
  n20 = ScalarParameter() {defaultValue=5, parameterName=BottomFadeMult}
  n21 = LinearInterpolate(B=Saturate(None=Multiply(A=@n19, B=@n20)), Alpha=ScalarParameter() {parameterName=UseBottomFade?}) {constA=1}
  n22 = CameraVectorWS()
  n23 = Transform(None=VectorParameter() {parameterName=FadeDirection, defaultValue=<rgba 0 1 0 1>}.RGB) {transformSourceType=TRANSFORMSOURCE_Instance}
  n24 = Multiply(
   A=Power(Base=DotProduct(A=@n22, B=@n23), Exp=ScalarParameter() {defaultValue=1, parameterName=ViewBasedFadeSensitivity, sliderMax=16}) {constExponent=1}
   B=ScalarParameter() {defaultValue=4, parameterName=FadeSensitivityMult}
  ) {constB=4}
  n25 = Multiply(
   A=Add(A=Multiply(A=@n17.R, B=@n21)) {constB=0}
   B=LinearInterpolate(B=Clamp(None=@n24), Alpha=ScalarParameter() {defaultValue=1, parameterName=UseAngleFade?}) {constA=1}
  )
  Multiply(A=Multiply(A=DepthFade(FadeDistance=@n16), B=Saturate(None=@n25)), B=ParticleColor().A)
  n26 = Add(
   A=MaterialFunctionCall(UVs=TextureCoordinate(), Texture Scale=ScalarParameter() {defaultValue=0.8, parameterName=CenterScale}) {materialFunction=&ScaleUVsByCenter}.UVs
   B=MaterialFunctionCall(X=ScalarParameter() {parameterName=UVOffsetX}, Y=ScalarParameter() {defaultValue=-0.1, parameterName=UVOffsetY}) {materialFunction=&MakeFloat2}.Result
  )
  n27 = MaterialFunctionCall(X=ScalarParameter() {defaultValue=1, parameterName=UVMultX}, Y=ScalarParameter() {defaultValue=1, parameterName=UVMultY}) {materialFunction=&MakeFloat2}
  n28 = TextureCoordinate()
  n29 = Add(
   A=Multiply(
    A=MaterialFunctionCall(Float2=@n28) {materialFunction=&BreakOutFloat2Components}.G
    B=Constant() {r=-1.25}
   )
   B=Multiply(A=Time(), B=Constant() {r=0.77})
  )
  n30 = Multiply(
   A=MaterialFunctionCall(X=Sine(None=@n29), Y=Constant()) {materialFunction=&MakeFloat2}.Result
   B=ScalarParameter() {defaultValue=0.2, parameterName=B_1}
  ) {constB=0.3}
  n31 = Constant() {r=0.4}
  n32 = Constant() {r=0.6}
  n33 = TextureCoordinate()
  n34 = LinearInterpolate(
   A=Constant()
   Alpha=SmoothStep(
    Min=@n31
    Max=Add(A=@n31, B=@n32)
    Value=MaterialFunctionCall(Float2=@n33) {materialFunction=&BreakOutFloat2Components}.G
   )
  )
  NamedRerouteDeclaration(None=Add(A=Multiply(A=@n26, B=@n27.Result), B=LinearInterpolate(B=@n30, Alpha=@n34))) {name=UVDistortion, nodeColor=<rgba 0 0.1094 1 1>}
  Reroute()
  VertexNormalWS()
 ]
}
