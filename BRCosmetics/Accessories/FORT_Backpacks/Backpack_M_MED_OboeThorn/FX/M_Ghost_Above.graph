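# Translucent unlit material for Niagara mesh particles in the OboeThorn backpack's ghost effect.
# Particle color feeds the shared MF_FX_EmissiveAdjustments function for emissive.
# A view-based fade takes the lesser absolute camera-vector dot product against a parameterized local axis and the vertex normal.
# That value is raised to a power, smoothstepped between min and max parameters, clamped and scaled by Overall Opacity, dimming edge-on views by default.
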Material /BRCosmetics/Accessories/FORT_Backpacks/Backpack_M_MED_OboeThorn/FX/M_Ghost_Above
graph {
 properties {
  bAutomaticallySetUsageInEditor=false
  bUsedWithNiagaraMeshParticles=true
  blendMode=BLEND_Translucent
  shadingModel=MSM_Unlit
  usedShadingModels=Unlit
 }
 nodes [
  MaterialFunctionCall(Emissive=ParticleColor().RGB) {materialFunction=&MF_FX_EmissiveAdjustments}
  n1 = VectorParameter() {parameterName=ViewBasedFadeLocalAxis, defaultValue=<rgba 0 0 1 1>}
  n2 = CameraVectorWS()
  n3 = Abs(None=DotProduct(A=Transform(None=@n1.RGB) {transformSourceType=TRANSFORMSOURCE_Local}, B=@n2))
  n4 = Abs(None=DotProduct(A=VertexNormalWS(), B=@n2))
  n5 = SmoothStep(
   Min=ScalarParameter() {parameterName=ViewBasedFadeMin, sliderMax=2, sliderMin=-2}
   Max=ScalarParameter() {defaultValue=1, parameterName=ViewBasedFadeMax, sliderMax=2, sliderMin=-2}
   Value=Power(Base=Min(A=@n3, B=@n4))
  )
  Multiply(A=ScalarParameter() {defaultValue=1, parameterName=Overall Opacity, sliderMax=1}, B=Multiply(A=Clamp(None=@n5)) {constA=1}) {constA=1, constB=-1}
 ]
}
