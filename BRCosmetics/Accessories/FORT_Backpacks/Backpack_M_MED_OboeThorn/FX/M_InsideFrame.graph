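# Unlit translucent material for the inside-frame effect mesh of the OboeThorn backpack cosmetic, usable on Niagara mesh particles.
# A color parameter, teal by default, is tinted by particle color and passed through the shared emissive adjustments function.
# A gradient from inverted UV V, shaped by a power parameter and scaled by particle alpha, produces a fade along the mesh.
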Material /BRCosmetics/Accessories/FORT_Backpacks/Backpack_M_MED_OboeThorn/FX/M_InsideFrame
graph {
 properties {
  bAutomaticallySetUsageInEditor=false
  bUsedWithNiagaraMeshParticles=true
  blendMode=BLEND_Translucent
  previewMesh=&/SM_FX_OboeThorn_InsideFrame
  shadingModel=MSM_Unlit
  usedShadingModels=Unlit
 }
 nodes [
  n1 = ParticleColor()
  MaterialFunctionCall(Emissive=Multiply(A=VectorParameter() {parameterName=MainColor, defaultValue=<rgba 0.0145 0.2 0.1726 1>}.RGB, B=@n1.RGB)) {materialFunction=&MF_FX_EmissiveAdjustments}
  n2 = MaterialFunctionCall(Float2=TextureCoordinate()) {materialFunction=&BreakOutFloat2Components}
  Multiply(A=Power(Base=OneMinus(None=@n2.G), Exp=ScalarParameter() {defaultValue=2, parameterName=Power}), B=@n1.A)
 ]
}
