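# Niagara system for the OboeThorn back bling, adding flickering candle flames, ghost apparitions and ponytail glows.
# Flame particles follow per-candle offsets on the skeletal mesh, feed lagged motion into their material and dim in the front end.
# System reactivity and timeline values drive ghost bursts, while a weapon-on flag kills either outer or inside-helmet ghosts.
# Spawn groups select each ghost layer's look, with camera-dot-product offsets and squash-stretch curves.
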NiagaraSystem /BRCosmetics/Accessories/FORT_Backpacks/Backpack_M_MED_OboeThorn/FX/NS_Backpack_OboeThorn
dependencies {
 usedDataInterfaces [
  &NiagaraDataInterfaceCurve
  &NiagaraDataInterfaceCamera
  &NiagaraDataInterfaceArrayFloat3
  &NiagaraDataInterfaceSkeletalMesh
 ]
 usedDynamicInputs [
  &MakeCustomIntFromBool
  &Boolean_AndOperation
  &Subtract_Float
 ]
 usedModules [
  &self:SystemSpawnScript.SystemScriptSource.SystemScriptGraph.NiagaraNodeAssignment_9.SetVariables_#1
  &self:SystemSpawnScript.SystemScriptSource.SystemScriptGraph.NiagaraNodeAssignment_4.SetVariables_#2
  &SystemState
  /Game/Effects/Niagara/Modules/Logic/UpdateValueStateChange
  &self:SystemSpawnScript.SystemScriptSource.SystemScriptGraph.NiagaraNodeAssignment_10.SetVariables_#3
  &Timeline
  &self:SystemSpawnScript.SystemScriptSource.SystemScriptGraph.NiagaraNodeAssignment_3.SetVariables_#4
 ]
}
emitters {
 FlamesAndGlows {
  emitterUpdateScript [
   &EmitterState(
    Inactive Response=::NewEnumerator1
    Life Cycle Mode=::NewEnumerator1
    Loop Behavior=::NewEnumerator1
    Loop Duration Mode=::NewEnumerator1
    UseLoopDelay=false
   )
   &SpawnBurst_Instantaneous(Spawn Count=6, Use Spawn Probability=false, UseLoopCountLimit=false)
  ]
  particleSpawnScript [
   &InitializeParticle(
    Color=<rgba 5 0.1667 0 1>
    Lifetime Max=0.5
    Lifetime Min=0.1
    Lifetime Mode=::NewEnumerator1
    Mesh Scale=&MakeVector(
     X=&Multiply_Float(A=0.7, B=&RandomRangeFloat(Minimum=0.6, Maximum=0.8))
     Y=0.5
     Z=0.5
    )
    Mesh Scale Mode=::NewEnumerator4
    Sprite Size Mode=::NewEnumerator1
    Uniform Sprite Size Max=5
    Uniform Sprite Size Min=6
    UsePositionOffset=true
    Write Material Random=true
   )
   &FN_SetSpawnGroup(Number of Spawn Groups:NiagaraInt32=$System.NumCandles) {enabled=false}
   &PartitionParticles(Number Of Partitions=6)
   /SetVariables(Particles.LaggedPosition:Vector3f=&ConvertPositionToVector(Input Position=$Particles.Position, Conversion Operation:ENiagara_LWCConvertPosToVec=::NewEnumerator0)) {isSetParameters=true}
   /SetVariables(Particles.ColorAlt:LinearColor=<rgba 1 0.1601 0.1458 0.4>) {isSetParameters=true}
  ]
  particleUpdateScript [
   &ParticleState(Kill Particles When Lifetime Has Elapsed=false, Let Infinitely Lived Particles Die When Emitter Deactivates=true)
   &ScaleMeshSize(Scale Factor=&MakeVector(
     X=&Waveform(
      Global Input Position=&RandomRangeFloat(Minimum=0, Maximum=1)
      "Waveform [1]"=::NewEnumerator0
      "[1] Frequency"=1
      "[1] Amplitude Min/Max"=<0.7 1>
     )
     Y=1
     Z=1
    ))
   &SkeletalMeshLocation(
    Bone Sampling Mode=::NewEnumerator4
    Inherit Velocity=false
    Interpolate Mesh Sampling=false
    MaskSpawnGroup=false
    Sampled Position Offset=&Add_Vector(
     A=&SelectVectorFromArray(Direct Array Index=$Particles.SpawnGroup, Vector Selection Array=$System.CandleOffsets)
     B=&Add_Vector(A=<0 -2.5 0>, B=<0 0 0>)
    )
    Sampled Position Offset Coordinate Space:ENiagaraMeshTransforms=::NewEnumerator2
    Skeletal Mesh=$System.SkeletalMesh
   ) {enabled=false, module=SkeletalMeshLocation002}
   &SkeletalMeshLocation(
    Bone Sampling Mode=::NewEnumerator4
    Inherit Velocity=false
    Interpolate Mesh Sampling=false
    MaskSpawnGroup=false
    Sampled Position Offset=&Add_Vector(
     A=&SelectVectorFromArray(Direct Array Index=$Particles.PartitionParticles.Partition, Vector Selection Array=$System.CandleOffsets)
     B=&Add_Vector(A=<0 -2.5 0>, B=<0 0 0>)
    )
    Sampled Position Offset Coordinate Space:ENiagaraMeshTransforms=::NewEnumerator2
    Skeletal Mesh=$System.SkeletalMesh
   )
   &SolveForcesAndVelocity(Clamp Velocity=false, Limit Acceleration=false)
   &ScaleColor(
    Scale Alpha=&Waveform(
     Global Input Position=&RandomRangeFloat(Minimum=0, Maximum=1)
     Global Phase=$System.NormalizedLoopAge
     "Waveform [1]"=::NewEnumerator12
     "[1] Frequency"=1
     "[1] Amplitude Min/Max"=<0 1>
    )
    Scale RGB=&VectorFromFloat(Value=&Waveform(
      Global Input Position=&RandomRangeFloat(Minimum=0, Maximum=1)
      Global Phase=$Particles.NormalizedAge
      "Waveform [1]"=::NewEnumerator12
      "[1] Frequency"=2
      "[1] Amplitude Min/Max"=<1 2>
     ))
    ScaleA=true
    ScaleRGB=true
   )
   &ScaleSpriteSize(Scale Factor=&Multiply_Vector2D_ByFloat(
     Float=&Waveform("Waveform [1]"=::NewEnumerator12, "[1] Frequency"=5, "[1] Amplitude Min/Max"=<1 2>, "[1] Scale"=0.5, "[1] Bias"=0.5)
     Vector2D=<1 1>
    ))
   /SetVariables(
    Particles.ColorAlt:LinearColor=&MultiplyLinearColor(
     Linear Color=$Particles.Initial.ColorAlt
     Scale Factor=&MakeVector4_FromFloat(Float=&Waveform(
       "Waveform [1]"=::NewEnumerator12
       "[1] Input Position"=&ReturnNormalizedExecIndex(Normalized Index Scale=0.8)
       "[1] Frequency"=3
       "[1] Amplitude Min/Max"=<0.5 1>
      ))
    )
    Particles.SpriteSizeAlt:Vector2f=&Vector2DFromFloat(Value=&FloatFromCurve(
      FloatCurve=DataInterface() {propertyValues {Curve {keys [
          {arriveTangent=-0.0036, interpMode=RCIM_Cubic, leaveTangent=-0.0036, tangentMode=RCTM_User, time=0, value=1}
          {arriveTangent=-2.2201, interpMode=RCIM_Cubic, leaveTangent=-2.2201, tangentMode=RCTM_User, time=1, value=0}
         ]}}}
      Scale Curve=&RandomRangeFloat(Minimum=5, Maximum=10)
     ))
   ) {isSetParameters=true}
   /SetVariables(Particles.MaterialRandom:NiagaraFloat=&RandomRangeFloat(Minimum=0, Maximum=1)) {isSetParameters=true}
   /SetVariables(Particles.LaggedPosition:Vector3f=/Niagara/DynamicInputs/Lerp/SmoothLerpOverTime_Position(
     Convergence Rate:NiagaraFloat=28
     Smooth Value:NiagaraPosition=Linked() {linkedVariable {name=Particles.LaggedPosition, type=Vector3f}}
     Target Value:NiagaraPosition=$Particles.Position
    )) {isSetParameters=true}
   /SetVariables(Particles.PositionDelta:Vector3f=&Subtract_Vector(A=$Particles.LaggedPosition, B=Linked() {linkedVariable {name=Particles.Position, type=NiagaraPosition}})) {isSetParameters=true}
   &ScaleColor(
    Scale RGB=&VectorFromFloat(Value=&FN_FrontEndFloatModifier(Gameplay=1, Front End=0.15))
    ScaleA=false
    ScaleRGB=true
   ) {module=ScaleColor001}
   &DynamicMaterialParameters(
    Index 0 Param 1=&Lerp_Float(
     A=1
     B=2
     Alpha=&Waveform(
      Global Input Position=&RandomRangeFloat(Minimum=0, Maximum=1)
      "Waveform [1]"=::NewEnumerator12
      "[1] Frequency"=5
      "[1] Amplitude Min/Max"=<0 1>
     )
    )
    Index 0 Param 2=&MakeFloatFromVector(Vector=$Particles.PositionDelta, Channel=::NewEnumerator0)
    Index 0 Param 3=&MakeFloatFromVector(Vector=$Particles.PositionDelta, Channel=::NewEnumerator1)
    Index 0 Param 4=&ClampFloat(Float=&MakeFloatFromVector(Vector=$Particles.PositionDelta, Channel=::NewEnumerator2), Max=2)
    Module.Index0Param0Write=true
    Module.Index0Param1Write=true
    Module.Index0Param2Write=true
    Module.Index0Param3Write=true
   )
   &FN_Sprite_Attributes(
    Camera Offset=3
    Color=&MultiplyLinearColor(
     Linear Color=<rgba 1 0.25 0 1.3>
     Scale Factor=&MakeVector4FromVectorAndFloat(
      "Vector (RGB)"=&VectorFromFloat(Value=&Waveform(
        "Waveform [1]"=::NewEnumerator0
        "[1] Input Position"=&RandomRangeFloat(Minimum=0, Maximum=1)
        "[1] Frequency"=1
        "[1] Amplitude Min/Max"=<0.8 1.5>
       ))
      "Float (Alpha)"=&Multiply_Float(
       A=1
       B=&ClampFloat(Float=&DotProduct(Vector A=&ReturnCameraVector(Camera Vectors=::NewEnumerator1, Coordinate Space=::Local), Vector B=<0 0 -1>))
      )
     )
    )
    Create Camera Offset Attribute=true
    Create Color Attribute=true
    Create SpriteSize Attribute=true
    SpriteSize=&Vector2DFromFloat(Value=1.3)
    Write to Module Namespace=true
   )
   &FN_Sprite_Attributes(
    Camera Offset=3
    Color=&MultiplyLinearColor(
     Linear Color=<rgba 0.5 0.0583 0 1.8>
     Scale Factor=&MakeVector4FromVectorAndFloat(
      "Vector (RGB)"=<1 1 1>
      "Float (Alpha)"=&Multiply_Float(
       A=1
       B=&ClampFloat(Float=&DotProduct(Vector A=&ReturnCameraVector(Camera Vectors=::NewEnumerator1, Coordinate Space=::Local), Vector B=<0 0 -1>))
      )
     )
    )
    Create Camera Offset Attribute=true
    Create Color Attribute=true
    Create SpriteSize Attribute=true
    SpriteSize=&Vector2DFromFloat(Value=&Multiply_Float(
      A=3
      B=&Waveform(
       "Waveform [1]"=::NewEnumerator0
       "[1] Input Position"=&RandomRangeFloat(Minimum=0, Maximum=1)
       "[1] Frequency"=1
       "[1] Amplitude Min/Max"=<0.8 1.2>
      )
     ))
    Write to Module Namespace=true
   ) {module=FN_Sprite_Attributes001}
  ]
  rendererClasses [&NiagaraSpriteRendererProperties, &NiagaraMeshRendererProperties]
  renderers [
   {rendererClass=&NiagaraSpriteRendererProperties}
   {rendererClass=&NiagaraMeshRendererProperties, rendererIndex=1}
   {rendererClass=&NiagaraSpriteRendererProperties, rendererIndex=2}
   {rendererClass=&NiagaraSpriteRendererProperties, rendererIndex=3}
  ]
 }
 GhostBoosted {
  emitterUpdateScript [
   &EmitterState(Life Cycle Mode=::NewEnumerator1, Loop Behavior=::NewEnumerator1, Loop Duration Mode=::NewEnumerator1, UseLoopDelay=false)
   /SetVariables(Emitter.LoopedAge:NiagaraFloat=$System.ReactivityTiming) {isSetParameters=true}
   &SpawnBurst_Instantaneous(Spawn Time=0.01, Use Spawn Probability=false, UseLoopCountLimit=true)
   &SpawnBurst_Instantaneous(Spawn Group=1, Spawn Time=0.01, Use Spawn Probability=false, UseLoopCountLimit=true) {module=SpawnBurst_Instantaneous001}
   &SpawnBurst_Instantaneous(Spawn Group=2, Spawn Time=0.01, Use Spawn Probability=false, UseLoopCountLimit=true) {module=SpawnBurst_Instantaneous002}
   &SpawnBurst_Instantaneous(Spawn Group=3, Spawn Time=0.01, Use Spawn Probability=false, UseLoopCountLimit=true) {module=SpawnBurst_Instantaneous003}
  ]
  particleSpawnScript [
   /SetVariables(Particles.Partition:NiagaraInt32=&SpawnGroup()) {isSetParameters=true}
   /SetVariables(Particles.VisibilityTag=$StackContext.Partition) {isSetParameters=true}
   &InitializeParticle(
    Color=&MergeHelper_ColorSwitch(
     0=<rgba 1 1 1 0.95>
     1=<rgba 0.05 1 1 1>
     2=<rgba 1 1 1 0.07>
     3=<rgba 0.1302 1.666 1.666 1>
     4=<rgba 0 0 0 0>
     Index=$Particles.Partition
     Num=::NewEnumerator3
    )
    Lifetime=$User.ReactivityDuration
    Mesh Scale=&MergeHelper_VectorSwitch(
     0=&Multiply_Vector(A=<0.5 0.5 0.6>, B=&VectorFromFloat(Value=1.05))
     1=&Multiply_VectorByFloat(Vector=<2.5 2 1>, Float=0.1)
     2=&Multiply_VectorByFloat(Vector=<0.9 1.1 1>, Float=0.15)
     3=<0 0 0>
     4=<0 0 0>
     Index=$Particles.Partition
     Num=::NewEnumerator3
    )
    Mesh Scale Mode=::NewEnumerator4
    Position Offset=<0 -22 5>
    Sprite Rotation Mode=::NewEnumerator2
    Sprite Size Mode=::NewEnumerator3
    Uniform Sprite Size=15
    UsePositionOffset=false
    Write Material Random=true
   )
   &InitialMeshOrientation(
    Mesh Axis To Orient=&MergeHelper_VectorSwitch(0=<1 0 0>, 1=<1 0 0>, 2=<0 0 1>, 3=<0 0 0>, 4=<0 0 0>, Index=$Particles.Partition, Num=::NewEnumerator3)
    Mesh Orientation Mode=::NewEnumerator2
    Orientation Vector=&MergeHelper_VectorSwitch(0=<1 0 0>, 1=<1 0 0>, 2=<-1 0 0>, 3=<0 0 0>, 4=<0 0 0>, Index=$Particles.Partition, Num=::NewEnumerator3)
    Use Rotation Vector=true
   )
  ]
  particleUpdateScript [
   &ParticleState()
   &SkeletalMeshLocation(
    Bone Sampling Mode=::NewEnumerator4
    MaskSpawnGroup=false
    Sampled Position Offset=&MergeHelper_VectorSwitch(0=<0 -17 -3>, 1=<0 -20 8>, 2=<0 -16 -3>, 3=<0 -22 9>, 4=<0 0 0>, Index=$Particles.Partition, Num=::NewEnumerator3)
    Skeletal Mesh=$System.SkeletalMesh
   )
   /SetVariables(Particles.SideView:NiagaraFloat=&Abs_Float(Float:NiagaraFloat=&DotProduct(Vector A=&ReturnCameraVector(Camera Vectors=::NewEnumerator1, Coordinate Space=::Local), Vector B=<1 0 0>))) {isSetParameters=true}
   &ScaleMeshSize(Scale Factor=&VectorFromFloat(Value=&Lerp_Float(A=1, B=0.9, Alpha=$Particles.SideView))) {module=ScaleMeshSize003}
   /SetVariables(Particles.Position=&AddVectorToPosition(
     Position=$Particles.Position
     Vector=&MergeHelper_VectorSwitch(
      0=&Lerp_Vector(A=<0 0 0>, B=<0 -3 0>, Alpha=$Particles.SideView)
      1=<0 0 0>
      2=<0 0 0>
      3=<0 0 0>
      4=<0 0 0>
      Index=$Particles.Partition
      Num=::NewEnumerator3
     )
    )) {isSetParameters=true}
   &ScaleColor(
    Scale Alpha=&FloatFromCurve(
     FloatCurve=DataInterface() {propertyValues {Curve {keys [
         {interpMode=RCIM_Cubic, leaveTangent=3.1026, tangentMode=RCTM_Break, time=0, value=0}
         {interpMode=RCIM_Cubic, tangentMode=RCTM_Break, time=0.25, value=1}
         {arriveTangentWeight=0.252, interpMode=RCIM_Cubic, leaveTangentWeight=0.1362, tangentMode=RCTM_User, time=0.8, value=1}
         <at 0.9 0 RCIM_Cubic>
        ]}}}
     Scale Curve=&MergeHelper_FloatSwitch(
      0=1
      1=1
      2=1
      3=&DotProduct(Vector A=&ReturnCameraVector(Camera Vectors=::NewEnumerator1, Coordinate Space=::Local), Vector B=<0 1 0>)
      4=1
      Index=$Particles.Partition
      Num=::NewEnumerator3
     )
    )
    ScaleA=true
    ScaleRGB=false
   ) {module=ScaleColor002}
   /SetVariables(Particles.Position=&AddVectorToPosition(
     Position=$Particles.Position
     Vector=&MergeHelper_VectorSwitch(
      0=&Lerp_Vector(
       A=<0 0 -3>
       B=<1 1 0>
       Alpha=&FloatFromCurve(FloatCurve=DataInterface() {propertyValues {Curve {keys [
            {interpMode=RCIM_Cubic, leaveTangent=8.9082, tangentMode=RCTM_Break, time=0, value=0}
            {interpMode=RCIM_Cubic, tangentMode=RCTM_Break, time=0.18, value=1}
            {interpMode=RCIM_Cubic, tangentMode=RCTM_Break, time=0.75, value=1}
            {arriveTangent=-8.9428, interpMode=RCIM_Cubic, tangentMode=RCTM_Break, time=0.9, value=0}
           ]}}})
      )
      1=<0 0 0>
      2=<0 0 0>
      3=<0 0 0>
      4=<0 0 0>
      Index=$Particles.Partition
      Num=::NewEnumerator3
     )
    )) {isSetParameters=true}
   &CameraOffset(Camera Offset Amount=&MergeHelper_FloatSwitch(
     0=&Add_Float(
      A=&Lerp_Float(
       A=1
       B=-10
       Alpha=&ClampFloat(Float=&DotProduct(Vector A=&ReturnCameraVector(Camera Vectors=::NewEnumerator1, Coordinate Space=::Local), Vector B=<0 -1 0>))
      )
      B=&Lerp_Float(A=0, B=-5, Alpha=$Particles.SideView)
     )
     1=3
     2=0
     3=0
     4=0
     Index=$StackContext.Partition
     Num=::NewEnumerator3
    ))
   &KillParticles(Kill Particles=&Boolean_NotOperation(Bool=$User.bIsWeaponOn))
  ]
  rendererClasses [&NiagaraMeshRendererProperties, &NiagaraSpriteRendererProperties]
  renderers [
   {rendererClass=&NiagaraMeshRendererProperties, rendererIndex=0}
   {rendererClass=&NiagaraMeshRendererProperties, rendererIndex=1}
   {rendererClass=&NiagaraMeshRendererProperties, rendererIndex=2}
   {rendererClass=&NiagaraSpriteRendererProperties, rendererIndex=3}
  ]
 }
 GhostBoosted_InsideHelmet {
  emitterUpdateScript [
   &EmitterState(Life Cycle Mode=::NewEnumerator1, Loop Behavior=::NewEnumerator1, Loop Duration Mode=::NewEnumerator1, UseLoopDelay=false)
   /SetVariables(Emitter.LoopedAge:NiagaraFloat=$System.ReactivityTiming) {isSetParameters=true}
   &SpawnBurst_Instantaneous(Spawn Time=0.01, Use Spawn Probability=false, UseLoopCountLimit=true)
   &SpawnBurst_Instantaneous(Spawn Group=1, Spawn Time=0.01, Use Spawn Probability=false, UseLoopCountLimit=true) {module=SpawnBurst_Instantaneous002}
  ]
  particleSpawnScript [
   /SetVariables(Particles.Partition:NiagaraInt32=&SpawnGroup()) {isSetParameters=true}
   /SetVariables(Particles.VisibilityTag=$StackContext.Partition) {isSetParameters=true}
   &InitializeParticle(
    Color=&MergeHelper_ColorSwitch(
     0=<rgba 1 1 1 1>
     1=<rgba 1 1 1 0.05>
     2=<rgba 0 0 0 0>
     3=<rgba 0 0 0 0>
     4=<rgba 0 0 0 0>
     Index=$Particles.Partition
     Num=::NewEnumerator3
    )
    Lifetime=$User.ReactivityDuration
    Mesh Scale=&MergeHelper_VectorSwitch(
     0=<1 1 1>
     1=&Multiply_VectorByFloat(Vector=<0.9 1.1 1>, Float=0.15)
     2=<0 0 0>
     3=<0 0 0>
     4=<0 0 0>
     Index=$Particles.Partition
     Num=::NewEnumerator3
    )
    Mesh Scale Mode=::NewEnumerator4
    Position Offset=<0 -22 5>
    Sprite Rotation Mode=::NewEnumerator2
    Sprite Size Mode=::NewEnumerator3
    Uniform Sprite Size=32
    UsePositionOffset=false
    Write Material Random=true
   )
   &InitialMeshOrientation(
    Mesh Axis To Orient=&MergeHelper_VectorSwitch(0=<1 0 0>, 1=<0 0 1>, 2=<0 0 1>, 3=<0 0 0>, 4=<0 0 0>, Index=$Particles.Partition, Num=::NewEnumerator3)
    Mesh Orientation Mode=::NewEnumerator2
    Orientation Vector=&MergeHelper_VectorSwitch(0=<1 0 0>, 1=<-1 0 0>, 2=<-1 0 0>, 3=<0 0 0>, 4=<0 0 0>, Index=$Particles.Partition, Num=::NewEnumerator3)
    Use Rotation Vector=true
   )
  ]
  particleUpdateScript [
   &ParticleState()
   &SkeletalMeshLocation(
    Bone / Socket Index=&MergeHelper_IntSwitch(0=6, 1=0, 2=0, 3=0, 4=0, 5=?, Index=$Particles.Partition, Num=::NewEnumerator3)
    Bone Sampling Mode=::NewEnumerator4
    MaskSpawnGroup=false
    Sampled Position Offset=&MergeHelper_VectorSwitch(0=<0 0 0>, 1=<0 -16 -3>, 2=<0 0 0>, 3=<0 0 0>, 4=<0 0 0>, Index=$Particles.Partition, Num=::NewEnumerator3)
    Skeletal Mesh=$System.SkeletalMesh
   )
   &ScaleColor(
    Scale Alpha=&FloatFromCurve(FloatCurve=DataInterface() {propertyValues {Curve {keys [
         {interpMode=RCIM_Cubic, leaveTangent=3.1026, tangentMode=RCTM_Break, time=0, value=0}
         {interpMode=RCIM_Cubic, tangentMode=RCTM_Break, time=0.222, value=1}
         {arriveTangentWeight=0.252, interpMode=RCIM_Cubic, leaveTangentWeight=0.1362, tangentMode=RCTM_User, time=0.8, value=1}
         <at 0.9 0 RCIM_Cubic>
        ]}}})
    ScaleA=true
    ScaleRGB=false
   ) {module=ScaleColor002}
   &KillParticles(Kill Particles=$User.bIsWeaponOn)
  ]
  rendererClasses [&NiagaraMeshRendererProperties]
  renderers [
   {rendererClass=&NiagaraMeshRendererProperties, rendererIndex=0}
   {rendererClass=&NiagaraMeshRendererProperties, rendererIndex=1}
   {rendererClass=&NiagaraMeshRendererProperties, rendererIndex=2}
  ]
 }
 GhostSmall {
  bEnabled=false
  emitterUpdateScript [
   &EmitterState(Life Cycle Mode=::NewEnumerator1, Loop Behavior=::NewEnumerator1, Loop Duration Mode=::NewEnumerator1, UseLoopDelay=false)
   /SetVariables(Emitter.LoopedAge:NiagaraFloat=$System.Timeline001.NormalizedTime) {isSetParameters=true, module=SetVariables_#5}
   &SpawnBurst_Instantaneous(Spawn Time=0.1, Use Spawn Probability=false, UseLoopCountLimit=true)
   &SpawnBurst_Instantaneous(Spawn Group=1, Spawn Time=0.1, Use Spawn Probability=false, UseLoopCountLimit=true) {module=SpawnBurst_Instantaneous001}
   &SpawnBurst_Instantaneous(Spawn Group=2, Spawn Time=0.1, Use Spawn Probability=false, UseLoopCountLimit=true) {module=SpawnBurst_Instantaneous002}
   &SpawnBurst_Instantaneous(Spawn Group=3, Spawn Time=0.1, Use Spawn Probability=false, UseLoopCountLimit=true) {module=SpawnBurst_Instantaneous003}
   &SpawnBurst_Instantaneous(Spawn Group=4, Spawn Time=0.1, Use Spawn Probability=false, UseLoopCountLimit=true) {module=SpawnBurst_Instantaneous004}
  ]
  particleSpawnScript [
   /SetVariables(Particles.Partition:NiagaraInt32=&SpawnGroup()) {isSetParameters=true, module=SetVariables_#6}
   &InitializeParticle(
    Color=&MergeHelper_ColorSwitch(
     0=<rgba 1 1 1 1.2>
     1=<rgba 0.2 0.7333 1 1>
     2=<rgba 2 17 20 0.01>
     3=<rgba 0 0.03 0.04 0.333>
     4=<rgba 1.5625 20 20 1>
     Index=$StackContext.Partition
     Num=::NewEnumerator3
    )
    Lifetime=&MergeHelper_FloatSwitch(0=2, 1=2, 2=2, 3=2, 4=2, Index=$StackContext.Partition, Num=::NewEnumerator3)
    Mesh Scale=&MergeHelper_VectorSwitch(
     0=&Multiply_Vector(A=<0.5 0.5 0.6>, B=&VectorFromFloat(Value=0.6))
     1=&Multiply_VectorByFloat(Vector=<2.5 2 1>, Float=0.1)
     2=<0 0 0>
     3=<0 0 0>
     4=<0 0 0>
     Index=$StackContext.Partition
     Num=::NewEnumerator3
    )
    Mesh Scale Mode=::NewEnumerator4
    Position Offset=<0 -22 5>
    Sprite Rotation Mode=::NewEnumerator2
    Sprite Size=&MergeHelper_Vector2Switch(0=<0 0>, 1=<0 0>, 2=<50 50>, 3=<60 60>, 4=<0 0>, Index=$StackContext.Partition, Num=::NewEnumerator3)
    Sprite Size Mode=::NewEnumerator4
    UsePositionOffset=false
    Write Material Random=true
   )
   /SetVariables(Particles.VisibilityTag=$StackContext.Partition) {isSetParameters=true, module=SetVariables_#7}
   &CameraOffset(Camera Offset Amount=&MergeHelper_FloatSwitch(0=0, 1=3, 2=5, 3=0, 4=0, Index=$StackContext.Partition, Num=::NewEnumerator3))
  ]
  particleUpdateScript [
   &ParticleState()
   &SkeletalMeshLocation(
    Bone Sampling Mode=::NewEnumerator4
    MaskSpawnGroup=false
    Sampled Position Offset=&MergeHelper_VectorSwitch(0=<0 -17 22>, 1=<0 -20 8>, 2=<0 -22 15>, 3=<0 -22 15>, 4=<0 -22 8>, Index=$StackContext.Partition, Num=::NewEnumerator3)
    Skeletal Mesh=$System.SkeletalMesh
   )
   /SetVariables(Particles.SideView:NiagaraFloat=&Abs_Float(Float:NiagaraFloat=&DotProduct(Vector A=&ReturnCameraVector(Camera Vectors=::NewEnumerator1, Coordinate Space=::Local), Vector B=<1 0 0>))) {isSetParameters=true, module=SetVariables_#8}
   /SetVariables(Particles.Position=&AddVectorToPosition(
     Position=$Particles.Position
     Vector=&Lerp_Vector(
      A=<0 0 0>
      B=&MergeHelper_VectorSwitch(0=<0 -6 0>, 1=<0 0 0>, 2=<0 0 0>, 3=<0 0 0>, 4=<0 0 0>, Index=$StackContext.Partition, Num=::NewEnumerator3)
      Alpha=$Particles.SideView
     )
    )) {isSetParameters=true, module=SetVariables_#9}
   /SetVariables(Particles.Position=&AddVectorToPosition(
     Position=$Particles.Position
     Vector=&MergeHelper_VectorSwitch(
      0=&Lerp_Vector(
       A=<0 0 -8>
       B=<0 0 0>
       Alpha=&FloatFromCurve(FloatCurve=DataInterface() {propertyValues {Curve {keys [
            {interpMode=RCIM_Cubic, leaveTangent=12.2141, tangentMode=RCTM_Break, time=0.0208, value=-0.3}
            {
             arriveTangent=-0.0729
             arriveTangentWeight=0.0515
             interpMode=RCIM_Cubic
             leaveTangent=-0.0729
             leaveTangentWeight=0.052
             tangentMode=RCTM_User
             time=0.1483
             value=1.25
            }
            {interpMode=RCIM_Cubic, tangentMode=RCTM_Break, time=0.3043, value=0.9139}
            {arriveTangentWeight=0.0436, interpMode=RCIM_Cubic, leaveTangentWeight=0.0583, tangentMode=RCTM_User, time=0.4655, value=1}
            {arriveTangent=-0.2931, interpMode=RCIM_Cubic, leaveTangent=-0.2931, time=0.6505, value=1}
            {arriveTangentWeight=0.0569, interpMode=RCIM_Cubic, leaveTangentWeight=0.0257, tangentMode=RCTM_User, time=0.7594, value=0.9139}
            {
             arriveTangent=1.1381
             arriveTangentWeight=0.0857
             interpMode=RCIM_Cubic
             leaveTangent=1.1381
             leaveTangentWeight=0.0857
             tangentMode=RCTM_User
             tangentWeightMode=RCTWM_WeightedBoth
             time=0.8916
             value=1.5656
            }
            {arriveTangent=-18.351, interpMode=RCIM_Cubic, leaveTangent=-18.351, tangentMode=RCTM_User, time=1, value=-0.0949}
           ]}}})
      )
      1=<0 0 0>
      2=<0 0 0>
      3=<0 0 0>
      4=<0 0 0>
      Index=$StackContext.Partition
      Num=::NewEnumerator3
     )
    )) {isSetParameters=true, module=SetVariables_#10}
   &ScaleMeshSize(Scale Factor=&MakeVector(
     X=&FloatFromCurve(FloatCurve=DataInterface() {propertyValues {Curve {keys [
          {interpMode=RCIM_Cubic, leaveTangent=6.2754, tangentMode=RCTM_Break, time=0.1036, value=0.4}
          {interpMode=RCIM_Cubic, tangentMode=RCTM_Break, time=0.1762, value=1.1}
          {arriveTangentWeight=0.0501, interpMode=RCIM_Cubic, leaveTangentWeight=0.0384, tangentMode=RCTM_User, time=0.3099, value=0.9881}
          {arriveTangent=0.0628, interpMode=RCIM_Cubic, leaveTangent=0.0628, time=0.4038, value=0.9993}
          {arriveTangentWeight=0.0384, interpMode=RCIM_Cubic, leaveTangentWeight=0.0501, tangentMode=RCTM_User, time=0.5, value=1}
          <at 0.7077 0.9993 RCIM_Cubic>
          {arriveTangentWeight=0.0262, interpMode=RCIM_Cubic, leaveTangentWeight=0.05, tangentMode=RCTM_User, time=0.7707, value=1}
          {arriveTangentWeight=0.045, interpMode=RCIM_Cubic, leaveTangentWeight=0.0104, tangentMode=RCTM_User, time=0.9108, value=1.3}
          {arriveTangent=-8.0521, interpMode=RCIM_Cubic, leaveTangent=-8.0521, tangentMode=RCTM_User, time=0.9682, value=0.4}
         ]}}})
     Y=1
     Z=&FloatFromCurve(FloatCurve=DataInterface() {propertyValues {Curve {keys [
          {arriveTangent=-7.97, interpMode=RCIM_Cubic, leaveTangent=-7.97, tangentMode=RCTM_User, time=0.1036, value=1.3}
          {interpMode=RCIM_Cubic, tangentMode=RCTM_User, time=0.1908, value=0.85}
          {arriveTangentWeight=0.0487, interpMode=RCIM_Cubic, leaveTangentWeight=0.1405, tangentMode=RCTM_User, time=0.3288, value=1.0329}
          {arriveTangentWeight=0.0319, interpMode=RCIM_Cubic, leaveTangentWeight=0.0628, tangentMode=RCTM_User, time=0.4475, value=1}
          {arriveTangentWeight=0.0633, interpMode=RCIM_Cubic, leaveTangentWeight=0.0321, tangentMode=RCTM_User, time=0.6684, value=1}
          {arriveTangentWeight=0.0319, interpMode=RCIM_Cubic, leaveTangentWeight=0.0463, tangentMode=RCTM_User, time=0.7537, value=1}
          {
           arriveTangentWeight=0.0463
           interpMode=RCIM_Cubic
           leaveTangentWeight=0.0293
           tangentMode=RCTM_User
           tangentWeightMode=RCTWM_WeightedBoth
           time=0.8899
           value=0.6
          }
          {arriveTangent=5.8777, tangentMode=RCTM_User, time=0.9623, value=1.4}
         ]}}})
    ))
   &ScaleColor(
    Scale Alpha=&FloatFromCurve(
     FloatCurve=DataInterface() {propertyValues {Curve {keys [
         {interpMode=RCIM_Cubic, leaveTangent=3.1026, tangentMode=RCTM_Break, time=0, value=0}
         {interpMode=RCIM_Cubic, tangentMode=RCTM_Break, time=0.4, value=1}
         {arriveTangentWeight=0.252, interpMode=RCIM_Cubic, leaveTangentWeight=0.1362, tangentMode=RCTM_User, time=0.8373, value=1}
         {arriveTangent=-8.256, interpMode=RCIM_Cubic, leaveTangent=-8.256, tangentMode=RCTM_User, time=0.9991, value=0}
        ]}}}
     Scale Curve=&MergeHelper_FloatSwitch(
      0=1
      1=1
      2=1
      3=1
      4=&Waveform(
       Waveform Count=::NewEnumerator1
       "Waveform [1]"=::NewEnumerator12
       "[1] Frequency"=1
       "[1] Amplitude Min/Max"=<0.1 1>
       "[2] Average"=false
       "Waveform [2]"=::NewEnumerator12
       "[2] Input Position"=0
       "[2] Phase"=$Emitter.Age
       "[2] Phase Scale"=1
       "[2] Compound Frequency"=2
       "[2] Frequency"=5
       "[2] Amplitude Min/Max"=<0.5 1.5>
       "[2] Scale"=1
       "[2] Bias"=0
       "[2] Exponent"=1
       "[2] Pulse Duration"=0.1
       "[2] Clip Triangle Min/Max"=<0 0>
       "[2] Random Seed"=0
      )
      Index=$StackContext.Partition
      Num=::NewEnumerator3
     )
    )
    ScaleA=true
    ScaleRGB=false
   ) {module=ScaleColor002}
   &ScaleColor(
    Scale Alpha=&MakeCustomFloatFromBool(Bool=$User.bIsWeaponOn, True Float=1, False Float=0)
    ScaleA=true
    ScaleRGB=false
   ) {module=ScaleColor001}
   &KillParticles(Kill Particles=&Boolean_NotOperation(Bool=$User.bIsWeaponOn))
   &FN_Light_Attributes(
    Color=$Particles.Color
    Create Color Attribute=true
    Create Exponent Attribute=true
    Create Radius Attribute=true
    Exponent=2
    Radius=75
    Scale Color=&FNTimeOfDayScalar(Night Scale=0.33, Morning Scale=0.5, Day Scale=0.5, Evening Scale=0.5)
   )
   &ScaleColor(
    Scale Alpha=&FloatFromCurve(FloatCurve=DataInterface() {propertyValues {Curve {keys [<at 0 0>,<at 0.05 1>]}}}, CurveIndex=$System.ReactivityTiming)
    ScaleA=true
    ScaleRGB=false
   )
  ]
  rendererClasses [
   &NiagaraMeshRendererProperties
   &NiagaraSpriteRendererProperties
   &NiagaraLightRendererProperties
  ]
  renderers [
   {rendererClass=&NiagaraMeshRendererProperties, rendererIndex=0}
   {rendererClass=&NiagaraMeshRendererProperties, rendererIndex=1}
   {rendererClass=&NiagaraSpriteRendererProperties, rendererIndex=2}
   {rendererClass=&NiagaraSpriteRendererProperties, rendererIndex=3}
   {rendererClass=&NiagaraLightRendererProperties, rendererIndex=4}
  ]
 }
 GhostSmall001 {
  emitterUpdateScript [
   &EmitterState(Life Cycle Mode=::NewEnumerator1, Loop Behavior=::NewEnumerator1, Loop Duration Mode=::NewEnumerator1, UseLoopDelay=false)
   /SetVariables(Emitter.LoopedAge:NiagaraFloat=$System.Timeline001.NormalizedTime) {isSetParameters=true, module=SetVariables_#5}
   &SpawnBurst_Instantaneous(Spawn Time=0.1, Use Spawn Probability=false, UseLoopCountLimit=true)
   &SpawnBurst_Instantaneous(Spawn Group=1, Spawn Time=0.1, Use Spawn Probability=false, UseLoopCountLimit=true) {module=SpawnBurst_Instantaneous001}
   &SpawnBurst_Instantaneous(Spawn Group=2, Spawn Time=0.1, Use Spawn Probability=false, UseLoopCountLimit=true) {module=SpawnBurst_Instantaneous002}
   &SpawnBurst_Instantaneous(Spawn Group=3, Spawn Time=0.1, Use Spawn Probability=false, UseLoopCountLimit=true) {module=SpawnBurst_Instantaneous003}
   &SpawnBurst_Instantaneous(Spawn Group=4, Spawn Time=0.1, Use Spawn Probability=false, UseLoopCountLimit=true) {module=SpawnBurst_Instantaneous004}
  ]
  particleSpawnScript [
   /SetVariables(Particles.Partition:NiagaraInt32=&SpawnGroup()) {isSetParameters=true, module=SetVariables_#6}
   &InitializeParticle(
    Color=&MergeHelper_ColorSwitch(
     0=<rgba 1 1 1 1.2>
     1=<rgba 0.2 0.7333 1 1>
     2=<rgba 2 17 20 0.01>
     3=<rgba 0 0.03 0.04 0.333>
     4=<rgba 1.5625 20 20 1>
     Index=$StackContext.Partition
     Num=::NewEnumerator3
    )
    Lifetime=&MergeHelper_FloatSwitch(0=1.2, 1=1.2, 2=1.2, 3=1.2, 4=1.2, Index=$StackContext.Partition, Num=::NewEnumerator3)
    Mesh Scale=&MergeHelper_VectorSwitch(
     0=&Multiply_Vector(A=<0.5 0.5 0.5>, B=&VectorFromFloat(Value=0.9))
     1=&Multiply_VectorByFloat(Vector=<2.5 2 1>, Float=0.1)
     2=<0 0 0>
     3=<0 0 0>
     4=<0 0 0>
     Index=$StackContext.Partition
     Num=::NewEnumerator3
    )
    Mesh Scale Mode=::NewEnumerator4
    Position Offset=<0 -22 5>
    Sprite Rotation Mode=::NewEnumerator2
    Sprite Size=&MergeHelper_Vector2Switch(0=<0 0>, 1=<0 0>, 2=<50 50>, 3=<60 60>, 4=<0 0>, Index=$StackContext.Partition, Num=::NewEnumerator3)
    Sprite Size Mode=::NewEnumerator4
    UsePositionOffset=false
    Write Material Random=true
   )
   /SetVariables(Particles.VisibilityTag=$StackContext.Partition) {isSetParameters=true, module=SetVariables_#7}
   &CameraOffset(Camera Offset Amount=&MergeHelper_FloatSwitch(0=0, 1=3, 2=5, 3=0, 4=0, Index=$StackContext.Partition, Num=::NewEnumerator3))
  ]
  particleUpdateScript [
   &ParticleState()
   &SkeletalMeshLocation(
    Bone Sampling Mode=::NewEnumerator4
    MaskSpawnGroup=false
    Sampled Position Offset=&MergeHelper_VectorSwitch(0=<0 -17 22>, 1=<0 -20 8>, 2=<0 -22 15>, 3=<0 -22 15>, 4=<0 -22 8>, Index=$StackContext.Partition, Num=::NewEnumerator3)
    Skeletal Mesh=$System.SkeletalMesh
   )
   /SetVariables(Particles.SideView:NiagaraFloat=&Abs_Float(Float:NiagaraFloat=&DotProduct(Vector A=&ReturnCameraVector(Camera Vectors=::NewEnumerator1, Coordinate Space=::Local), Vector B=<1 0 0>))) {isSetParameters=true, module=SetVariables_#8}
   /SetVariables(Particles.Position=&AddVectorToPosition(
     Position=$Particles.Position
     Vector=&Lerp_Vector(
      A=<0 0 0>
      B=&MergeHelper_VectorSwitch(0=<0 -6 0>, 1=<0 0 0>, 2=<0 0 0>, 3=<0 0 0>, 4=<0 0 0>, Index=$StackContext.Partition, Num=::NewEnumerator3)
      Alpha=$Particles.SideView
     )
    )) {isSetParameters=true, module=SetVariables_#9}
   /SetVariables(Particles.Position=&AddVectorToPosition(
     Position=$Particles.Position
     Vector=&MergeHelper_VectorSwitch(
      0=&Lerp_Vector(
       A=<0 0 -8>
       B=<0 0 4>
       Alpha=&FloatFromCurve(FloatCurve=DataInterface() {propertyValues {Curve {keys [
            {interpMode=RCIM_Cubic, leaveTangent=12.2141, tangentMode=RCTM_Break, time=0, value=-0.3}
            {
             arriveTangent=-0.0729
             arriveTangentWeight=0.0515
             interpMode=RCIM_Cubic
             leaveTangent=-0.0729
             leaveTangentWeight=0.052
             tangentMode=RCTM_User
             time=0.2421
             value=1.25
            }
            {interpMode=RCIM_Cubic, tangentMode=RCTM_Break, time=0.3632, value=0.9688}
            {arriveTangent=-0.3011, interpMode=RCIM_Cubic, leaveTangent=-0.3011, time=0.5, value=1}
            {arriveTangentWeight=0.0569, interpMode=RCIM_Cubic, leaveTangentWeight=0.0257, tangentMode=RCTM_User, time=0.5457, value=0.9139}
            {
             arriveTangent=1.1381
             arriveTangentWeight=0.0857
             interpMode=RCIM_Cubic
             leaveTangent=1.1381
             leaveTangentWeight=0.0857
             tangentMode=RCTM_User
             tangentWeightMode=RCTWM_WeightedBoth
             time=0.7247
             value=1.5656
            }
            {arriveTangent=-18.351, interpMode=RCIM_Cubic, leaveTangent=-18.351, tangentMode=RCTM_User, time=0.9038, value=-0.0949}
           ]}}})
      )
      1=<0 0 0>
      2=<0 0 0>
      3=<0 0 0>
      4=<0 0 0>
      Index=$StackContext.Partition
      Num=::NewEnumerator3
     )
    )) {isSetParameters=true, module=SetVariables_#10}
   &ScaleMeshSize(Scale Factor=&MakeVector(
     X=&FloatFromCurve(FloatCurve=DataInterface() {propertyValues {Curve {keys [
          {interpMode=RCIM_Cubic, leaveTangent=6.2754, tangentMode=RCTM_Break, time=0.1211, value=0.4}
          {interpMode=RCIM_Cubic, tangentMode=RCTM_Break, time=0.2421, value=1.1}
          {arriveTangentWeight=0.0501, interpMode=RCIM_Cubic, leaveTangentWeight=0.0384, tangentMode=RCTM_User, time=0.3632, value=0.991}
          {arriveTangent=0.0656, interpMode=RCIM_Cubic, leaveTangent=0.0656, time=0.4138, value=0.9993}
          {arriveTangentWeight=0.0384, interpMode=RCIM_Cubic, leaveTangentWeight=0.0501, tangentMode=RCTM_User, time=0.5, value=1}
          {arriveTangentWeight=0.0262, interpMode=RCIM_Cubic, leaveTangentWeight=0.05, tangentMode=RCTM_User, time=0.5405, value=1}
          {arriveTangentWeight=0.045, interpMode=RCIM_Cubic, leaveTangentWeight=0.0104, tangentMode=RCTM_User, time=0.782, value=1.2}
          {arriveTangent=-8.0521, interpMode=RCIM_Cubic, leaveTangent=-8.0521, tangentMode=RCTM_User, time=0.8469, value=0.4}
         ]}}})
     Y=1
     Z=&FloatFromCurve(FloatCurve=DataInterface() {propertyValues {Curve {keys [
          {arriveTangent=-7.97, interpMode=RCIM_Cubic, leaveTangent=-7.97, tangentMode=RCTM_User, time=0.1211, value=1.3}
          {interpMode=RCIM_Cubic, tangentMode=RCTM_User, time=0.2421, value=0.85}
          {arriveTangentWeight=0.0487, interpMode=RCIM_Cubic, leaveTangentWeight=0.1405, tangentMode=RCTM_User, time=0.3632, value=1.0329}
          {arriveTangentWeight=0.0319, interpMode=RCIM_Cubic, leaveTangentWeight=0.0628, tangentMode=RCTM_User, time=0.5, value=1}
          {arriveTangentWeight=0.0319, interpMode=RCIM_Cubic, leaveTangentWeight=0.0463, tangentMode=RCTM_User, time=0.5405, value=1}
          {
           arriveTangentWeight=0.0463
           interpMode=RCIM_Cubic
           leaveTangentWeight=0.0293
           tangentMode=RCTM_User
           tangentWeightMode=RCTWM_WeightedBoth
           time=0.7091
           value=0.6
          }
          {arriveTangent=5.8777, tangentMode=RCTM_User, time=0.7935, value=1.4}
         ]}}})
    ))
   &ScaleColor(
    Scale Alpha=&FloatFromCurve(
     FloatCurve=DataInterface() {propertyValues {Curve {keys [
         {interpMode=RCIM_Cubic, leaveTangent=3.1026, tangentMode=RCTM_Break, time=0, value=0}
         {interpMode=RCIM_Cubic, tangentMode=RCTM_Break, time=0.4, value=1}
         {arriveTangentWeight=0.252, interpMode=RCIM_Cubic, leaveTangentWeight=0.1362, tangentMode=RCTM_User, time=0.7512, value=1}
         {arriveTangent=-8.256, interpMode=RCIM_Cubic, leaveTangent=-8.256, tangentMode=RCTM_User, time=0.9131, value=0}
        ]}}}
     Scale Curve=&MergeHelper_FloatSwitch(
      0=1
      1=1
      2=1
      3=1
      4=&Waveform(
       Waveform Count=::NewEnumerator1
       "Waveform [1]"=::NewEnumerator12
       "[1] Frequency"=1
       "[1] Amplitude Min/Max"=<0.1 1>
       "[2] Average"=false
       "Waveform [2]"=::NewEnumerator12
       "[2] Input Position"=0
       "[2] Phase"=$Emitter.Age
       "[2] Phase Scale"=1
       "[2] Compound Frequency"=2
       "[2] Frequency"=5
       "[2] Amplitude Min/Max"=<0.5 1.5>
       "[2] Scale"=1
       "[2] Bias"=0
       "[2] Exponent"=1
       "[2] Pulse Duration"=0.1
       "[2] Clip Triangle Min/Max"=<0 0>
       "[2] Random Seed"=0
      )
      Index=$StackContext.Partition
      Num=::NewEnumerator3
     )
    )
    ScaleA=true
    ScaleRGB=false
   ) {module=ScaleColor002}
   &ScaleColor(
    Scale Alpha=&MakeCustomFloatFromBool(Bool=$User.bIsWeaponOn, True Float=1, False Float=0)
    ScaleA=true
    ScaleRGB=false
   ) {module=ScaleColor001}
   &KillParticles(Kill Particles=&Boolean_NotOperation(Bool=$User.bIsWeaponOn))
   &FN_Light_Attributes(
    Color=$Particles.Color
    Create Color Attribute=true
    Create Exponent Attribute=true
    Create Radius Attribute=true
    Exponent=2
    Radius=75
    Scale Color=&FNTimeOfDayScalar(Night Scale=0.33, Morning Scale=0.5, Day Scale=0.5, Evening Scale=0.5)
   )
   &ScaleColor(
    Scale Alpha=&FloatFromCurve(FloatCurve=DataInterface() {propertyValues {Curve {keys [<at 0 0>,<at 0.05 1>]}}}, CurveIndex=$System.ReactivityTiming)
    ScaleA=true
    ScaleRGB=false
   )
  ]
  rendererClasses [
   &NiagaraMeshRendererProperties
   &NiagaraSpriteRendererProperties
   &NiagaraLightRendererProperties
  ]
  renderers [
   {rendererClass=&NiagaraMeshRendererProperties, rendererIndex=0}
   {rendererClass=&NiagaraMeshRendererProperties, rendererIndex=1}
   {rendererClass=&NiagaraSpriteRendererProperties, rendererIndex=2}
   {rendererClass=&NiagaraSpriteRendererProperties, rendererIndex=3}
   {rendererClass=&NiagaraLightRendererProperties, rendererIndex=4}
  ]
 }
 PonyTailGlow {
  emitterUpdateScript [
   &EmitterState(
    Inactive Response=::NewEnumerator1
    Life Cycle Mode=::NewEnumerator1
    Loop Behavior=::NewEnumerator1
    Loop Duration Mode=::NewEnumerator1
    UseLoopDelay=false
   )
   /SetVariables(Emitter.LoopedAge:NiagaraFloat=$System.ReactivityTiming) {isSetParameters=true}
   &SpawnBurst_Instantaneous(Spawn Count=5, Spawn Time=0.001, Use Spawn Probability=false, UseLoopCountLimit=true)
   /SetVariables(Emitter.bool:NiagaraBool=&InvertBool(Boolean:NiagaraBool=$User.bIsFrontEndPreview)) {isSetParameters=true}
  ]
  particleSpawnScript [
   &PartitionParticles(Number Of Partitions=5)
   &InitializeParticle(
    Color Mode=::NewEnumerator0
    Lifetime=$User.ReactivityDuration
    Mesh Scale Mode=::NewEnumerator3
    Mesh Uniform Scale=&Multiply_Float(
     A=&MergeHelper_FloatSwitch(0=0, 1=0.04, 2=0.035, 3=0.04, 4=0.02, Index=$Particles.PartitionParticles.Partition, Num=::NewEnumerator3)
     B=1
    )
    UsePositionOffset=false
    Write Material Random=true
   )
   &CameraOffset(Camera Offset Amount=3)
  ]
  particleUpdateScript [
   &ParticleState()
   &SkeletalMeshLocation(
    Bone / Socket Index=&MergeHelper_IntSwitch(0=0, 1=1, 2=2, 3=3, 4=4, 5=?, Index=$Particles.PartitionParticles.Partition, Num=::NewEnumerator3)
    MaskSpawnGroup=false
    Mesh Sampling Type=::NewEnumerator5
    Orientation Sampling=::NewEnumerator1
    Skeletal Mesh=$System.SkeletalMesh
    Socket Sampling Mode=::NewEnumerator5
   )
   &ScaleColor(
    Scale Alpha=&FloatFromCurve(FloatCurve=DataInterface() {propertyValues {Curve {keys [
         {arriveTangent=5, interpMode=RCIM_Cubic, time=0, value=0}
         {arriveTangentWeight=0.1067, interpMode=RCIM_Cubic, leaveTangentWeight=0.3202, tangentMode=RCTM_User, time=0.222, value=1}
         {arriveTangentWeight=0.3202, interpMode=RCIM_Cubic, leaveTangentWeight=0.1067, tangentMode=RCTM_User, time=0.8, value=1}
         <at 1 0 RCIM_Cubic>
        ]}}})
    Scale RGB=&VectorFromFloat(Value=&Multiply_Float(A=1, B=&FNTimeOfDayScalar(Night Scale=0.8, Morning Scale=1, Evening Scale=1)))
    ScaleA=true
    ScaleRGB=true
   )
   &KillParticles(Kill Particles=$User.bIsWeaponOn)
  ]
  rendererClasses [&NiagaraMeshRendererProperties]
  renderers [
   {rendererClass=&NiagaraMeshRendererProperties, rendererIndex=0}
   {rendererClass=&NiagaraMeshRendererProperties, rendererIndex=1}
  ]
 }
 ReactivityGlows {
  emitterUpdateScript [
   &EmitterState(Life Cycle Mode=::NewEnumerator1, Loop Behavior=::NewEnumerator1, Loop Duration Mode=::NewEnumerator1, UseLoopDelay=false)
   /SetVariables(Emitter.LoopedAge:NiagaraFloat=$System.ReactivityTiming) {isSetParameters=true}
   &SpawnBurst_Instantaneous(Spawn Time=0.01, Use Spawn Probability=false, UseLoopCountLimit=true)
   &SpawnBurst_Instantaneous(Spawn Group=1, Spawn Time=0.01, Use Spawn Probability=false, UseLoopCountLimit=true) {module=SpawnBurst_Instantaneous001}
   &SpawnBurst_Instantaneous(Spawn Group=2, Spawn Time=0.01, Use Spawn Probability=false, UseLoopCountLimit=true) {module=SpawnBurst_Instantaneous002}
  ]
  particleSpawnScript [
   /SetVariables(Particles.Partition:NiagaraInt32=&SpawnGroup()) {isSetParameters=true}
   /SetVariables(Particles.VisibilityTag=$StackContext.Partition) {isSetParameters=true}
   &InitializeParticle(
    Color=&MergeHelper_ColorSwitch(
     0=<rgba 2 17 20 0.02>
     1=<rgba 0 0.03 0.04 0.333>
     2=<rgba 2.3438 30 30 1>
     3=<rgba 0 0 0 0>
     4=<rgba 0 0 0 0>
     Index=$Particles.Partition
     Num=::NewEnumerator3
    )
    Lifetime=$User.ReactivityDuration
    Sprite Size=&MergeHelper_Vector2Switch(
     0=&MakeCustomVector2DFromBool(Bool=$User.bIsWeaponOn, True Vector2D=<50 50>, False Vector2D=<35 35>)
     1=<60 60>
     2=<0 0>
     3=<0 0>
     4=<0 0>
     Index=$Particles.Partition
     Num=::NewEnumerator3
    )
    Sprite Size Mode=::NewEnumerator4
    UsePositionOffset=false
    Write Material Random=true
   )
   &CameraOffset(Camera Offset Amount=&MergeHelper_FloatSwitch(0=8, 1=0, 2=0, 3=0, 4=0, Index=$Particles.Partition, Num=::NewEnumerator3))
  ]
  particleUpdateScript [
   &ParticleState()
   &SkeletalMeshLocation(
    Bone Sampling Mode=::NewEnumerator4
    MaskSpawnGroup=false
    Sampled Position Offset=&MergeHelper_VectorSwitch(
     0=&MakeCustomVectorFromBool(Bool=$User.bIsWeaponOn, True Vector=<0 -22 15>, False Vector=<0 -24 6>)
     1=<0 -22 15>
     2=<0 -22 8>
     3=<0 0 0>
     4=<0 0 0>
     Index=$Particles.Partition
     Num=::NewEnumerator3
    )
    Skeletal Mesh=$System.SkeletalMesh
   )
   &ScaleColor(
    Scale Alpha=&FloatFromCurve(
     FloatCurve=DataInterface() {propertyValues {Curve {keys [
         {interpMode=RCIM_Cubic, leaveTangent=3.1026, tangentMode=RCTM_Break, time=0, value=0}
         {interpMode=RCIM_Cubic, tangentMode=RCTM_Break, time=0.222, value=1}
         {arriveTangentWeight=0.252, interpMode=RCIM_Cubic, leaveTangentWeight=0.1362, tangentMode=RCTM_User, time=0.8, value=1}
         <at 0.9 0 RCIM_Cubic>
        ]}}}
     Scale Curve=&MergeHelper_FloatSwitch(
      0=&FNTimeOfDayScalar(Night Scale=0.9, Morning Scale=0.75, Evening Scale=0.75)
      1=1
      2=&Waveform(
       Waveform Count=::NewEnumerator1
       "Waveform [1]"=::NewEnumerator12
       "[1] Frequency"=1
       "[1] Amplitude Min/Max"=<0.1 1>
       "[2] Average"=false
       "Waveform [2]"=::NewEnumerator12
       "[2] Input Position"=0
       "[2] Phase"=$Emitter.Age
       "[2] Phase Scale"=1
       "[2] Compound Frequency"=2
       "[2] Frequency"=5
       "[2] Amplitude Min/Max"=<0.5 1.5>
       "[2] Scale"=1
       "[2] Bias"=0
       "[2] Exponent"=1
       "[2] Pulse Duration"=0.1
       "[2] Clip Triangle Min/Max"=<0 0>
       "[2] Random Seed"=0
      )
      3=1
      4=1
      Index=$Particles.Partition
      Num=::NewEnumerator3
     )
    )
    ScaleA=true
    ScaleRGB=false
   ) {module=ScaleColor002}
   &FN_Light_Attributes(
    Color=$Particles.Color
    Create Color Attribute=true
    Create Exponent Attribute=true
    Create Radius Attribute=true
    Exponent=2
    Radius=75
    Scale Color=&FNTimeOfDayScalar(Night Scale=0.33, Morning Scale=0.5, Day Scale=0.5, Evening Scale=0.5)
   )
  ]
  rendererClasses [
   &NiagaraSpriteRendererProperties
   &NiagaraLightRendererProperties
   &NiagaraMeshRendererProperties
  ]
  renderers [
   {rendererClass=&NiagaraSpriteRendererProperties}
   {rendererClass=&NiagaraSpriteRendererProperties, rendererIndex=1}
   {rendererClass=&NiagaraLightRendererProperties, rendererIndex=2}
   {rendererClass=&NiagaraMeshRendererProperties, rendererIndex=3}
  ]
 }
 Trails_Combined {
  emitterUpdateScript [
   &EmitterState(
    Inactive Response=::NewEnumerator1
    Life Cycle Mode=::NewEnumerator1
    Loop Behavior=::NewEnumerator1
    Loop Duration Mode=::NewEnumerator1
    UseLoopDelay=false
   )
   /SetVariables(Emitter.LoopedAge:NiagaraFloat=$System.ReactivityTiming) {isSetParameters=true}
   &SpawnBurst_Instantaneous(Spawn Count=8, Spawn Time=0.001, Use Spawn Probability=false, UseLoopCountLimit=true)
  ]
  particleSpawnScript [
   &PartitionParticles(Number Of Partitions=8)
   &InitializeParticle(
    Color=&MaskLinearColor(
     Mask Source=1
     Mask Parameter=$Particles.PartitionParticles.Partition
     Mask=7
     Masked Linear Color=<rgba 1.5 1.5 1.5 0.8>
     Passthrough Linear Color=<rgba 1.5 1.5 1.5 1.2>
    )
    Lifetime=$User.ReactivityDuration
    Mesh Scale=&Multiply_VectorByFloat(
     Vector=&RandomRangeVector(Minimum=<0.8 1 0.8>, Maximum=<1.1 1 1>)
     Float=&MergeHelper_FloatSwitch(
      0=&Multiply_Float(
       A=&RandomRangeFloat(Minimum=0.05, Maximum=0.08)
       B=&MakeCustomFloatFromBool(Bool=$User.bIsWeaponOn, True Float=1, False Float=0.85)
      )
      1=&Multiply_Float(
       A=&RandomRangeFloat(Minimum=0.05, Maximum=0.08)
       B=&MakeCustomFloatFromBool(Bool=$User.bIsWeaponOn, True Float=1, False Float=0.95)
      )
      2=&Multiply_Float(
       A=&RandomRangeFloat(Minimum=0.05, Maximum=0.08)
       B=&MakeCustomFloatFromBool(Bool=$User.bIsWeaponOn, True Float=1, False Float=0.85)
      )
      3=&Multiply_Float(
       A=&RandomRangeFloat(Minimum=0.05, Maximum=0.08)
       B=&MakeCustomFloatFromBool(Bool=$User.bIsWeaponOn, True Float=1, False Float=0.95)
      )
      4=&Multiply_Float(A=&RandomRangeFloat(Minimum=0.07, Maximum=0.1), B=1)
      5=&Multiply_Float(A=&RandomRangeFloat(Minimum=0.07, Maximum=0.1), B=1)
      6=&Multiply_Float(A=&RandomRangeFloat(Minimum=0.07, Maximum=0.1), B=0.4)
      7=1.3
      Index=$Particles.PartitionParticles.Partition
      Num=::NewEnumerator6
     )
    )
    Mesh Scale Mode=::NewEnumerator4
    UsePositionOffset=false
    Write Material Random=true
   )
   &InitialMeshOrientation(
    Mesh Axis To Orient=&MaskVector(Mask Source=1, Mask Parameter=$Particles.PartitionParticles.Partition, Mask=7, Masked Vector=<0 1 0>, Passthrough Vector=<1 0 0>)
    Mesh Orientation Mode=::NewEnumerator2
    Orientation Vector=&MaskVector(Mask Source=1, Mask Parameter=$Particles.PartitionParticles.Partition, Mask=7, Masked Vector=<0 0 1>, Passthrough Vector=<1 0 0>)
    Rotation=&MergeHelper_VectorSwitch(
     0=<0 0 0.6>
     1=<0 0 0.6>
     2=<0 0 0>
     3=<0 0 0>
     4=<0 0 0>
     5=<0 0 0.5>
     6=<0 0 0.5>
     7=<0 0 0>
     Index=$Particles.PartitionParticles.Partition
     Num=::NewEnumerator6
    )
    Use Rotation Vector=true
   )
   &InitialMeshOrientation(
    Mesh Orientation Mode=::NewEnumerator2
    Rotation=&MergeHelper_VectorSwitch(
     0=<-0.06 0 0>
     1=<-0.035 0 0>
     2=<0.06 0 0>
     3=<0.035 0 0>
     4=<0 0 -0.02>
     5=<0 0 -0.05>
     6=<0 0 0.15>
     7=<0 0 0>
     Index=$Particles.PartitionParticles.Partition
     Num=::NewEnumerator6
    )
    Use Rotation Vector=true
   ) {module=InitialMeshOrientation002}
   /SetVariables(Particles.VisibilityTag=&MergeHelper_IntSwitch(0=0, 1=0, 2=0, 3=0, 4=0, 5=0, 6=0, 7=1, Index=$Particles.PartitionParticles.Partition, Num=::NewEnumerator6)) {isSetParameters=true}
  ]
  particleUpdateScript [
   &SkeletalMeshLocation(
    Bone Sampling Mode=::NewEnumerator4
    MaskSpawnGroup=false
    Sampled Position Offset=&MakeCustomVectorFromBool(
     Bool=$User.bIsWeaponOn
     True Vector=&MergeHelper_VectorSwitch(
      0=<-1 -21.8607 4.2609>
      1=<-3 -21.8607 8.068>
      2=<-1 -20 4.2609>
      3=<3 -18.6287 8.068>
      4=<11 -15 1>
      5=<-14 -12 0>
      6=<-13 -22 0>
      7=<0 -15 -9>
      Index=$Particles.PartitionParticles.Partition
      Num=::NewEnumerator6
     )
     False Vector=&MergeHelper_VectorSwitch(
      0=<-2 -21 4>
      1=<-4.5 -18.6287 6>
      2=<2 -21 4>
      3=<4.5 -18.6287 6>
      4=<11 -15 1>
      5=<-14 -12 0>
      6=<-13 -22 0>
      7=<0 -15 -9>
      Index=$Particles.PartitionParticles.Partition
      Num=::NewEnumerator6
     )
    )
    Skeletal Mesh=$System.SkeletalMesh
   )
   &ParticleState()
   &InterpolateOverTime(Rate Of Change=5, Target Vector Value:NiagaraPosition=$Particles.Position, Variable Type=::NewEnumerator2)
   /SetVariables(Particles.Interpolate DY:Vector4f=&MakeFloat4FromVectorAndFloat("Vector (RGB)"=$Particles.InterpolateOverTime.InterpolateOverTime_V3_.Moving Average, "Float (Alpha)"=0.15)) {isSetParameters=true}
   /SetVariables(Particles.ColorGlow:LinearColor=&MakeLinearColorFromVectorAndFloat(
     "Vector (RGB)"=&VectorFromFloat(Value=&FloatFromCurve(FloatCurve=DataInterface() {propertyValues {Curve {keys [{interpMode=RCIM_Cubic, tangentMode=RCTM_Break, time=0.7, value=1}, {interpMode=RCIM_Cubic, tangentMode=RCTM_Break, time=1, value=0}]}}}))
     "Float (Alpha)"=0.2
    )) {isSetParameters=true}
   &DynamicMaterialParameters(
    Index 0 Param 1=&Multiply_Float(
     A=&Multiply_Float(
      A=&MergeHelper_FloatSwitch(0=0.2, 1=0.3, 2=0.18, 3=0.3, 4=0.3, 5=0.3, 6=0.3, 7=0.35, Index=$Particles.PartitionParticles.Partition, Num=::NewEnumerator6)
      B=&RandomRangeFloat(
       Minimum=&MergeHelper_FloatSwitch(0=0.8, 1=0.8, 2=0.8, 3=0.8, 4=1, 5=1, 6=1, 7=1, Index=$Particles.PartitionParticles.Partition, Num=::NewEnumerator6)
       Maximum=&MergeHelper_FloatSwitch(0=0.9, 1=0.9, 2=0.9, 3=0.9, 4=1.4, 5=1.4, 6=1.4, 7=1.4, Index=$Particles.PartitionParticles.Partition, Num=::NewEnumerator6)
      )
     )
     B=&Multiply_Float(
      A=&FloatFromCurve(FloatCurve=DataInterface() {propertyValues {Curve {keys [
           {arriveTangent=4.6711, interpMode=RCIM_Cubic, leaveTangent=4.6711, tangentMode=RCTM_User, time=0, value=0}
           {arriveTangentWeight=0.1067, interpMode=RCIM_Cubic, leaveTangentWeight=0.3202, tangentMode=RCTM_User, time=0.3, value=1}
           {arriveTangentWeight=0.3202, interpMode=RCIM_Cubic, leaveTangentWeight=0.1067, tangentMode=RCTM_User, time=0.7, value=1}
           <at 1 0 RCIM_Cubic>
          ]}}})
      B=1
     )
    )
    Index 0 Param 2=&FloatFromCurve(
     FloatCurve=DataInterface() {propertyValues {Curve {keys [
         {arriveTangent=5, interpMode=RCIM_Cubic, time=0, value=0}
         {arriveTangentWeight=0.1067, interpMode=RCIM_Cubic, leaveTangentWeight=0.3202, tangentMode=RCTM_User, time=0.3, value=1}
         {arriveTangentWeight=0.3202, interpMode=RCIM_Cubic, leaveTangentWeight=0.1067, tangentMode=RCTM_User, time=0.7, value=1}
         <at 1 0 RCIM_Cubic>
        ]}}}
     Scale Curve=&RandomRangeFloat(Minimum=1, Maximum=1.4)
    )
    Index 0 Param 3=&Multiply_Float(
     A=&MergeHelper_FloatSwitch(0=0.9, 1=0.6, 2=0.8, 3=0.5, 4=1.5, 5=1.2, 6=1.6, 7=3, Index=$Particles.PartitionParticles.Partition, Num=::NewEnumerator6)
     B=1
    )
    Index 0 Param 4=&RandomRangeFloat(Minimum=0, Maximum=1)
    Index 2 Param 1=&MergeHelper_FloatSwitch(0=0.8, 1=0.6, 2=0.5, 3=0.6, 4=0.8, 5=0.5, 6=0.7, 7=1, Index=$Particles.PartitionParticles.Partition, Num=::NewEnumerator6)
    Module.Index0Param0Write=true
    Module.Index0Param1Write=true
    Module.Index0Param2Write=true
    Module.Index0Param3Write=true
    Module.Index2Param0Write=true
    Module.Index2Param1Write=true
    Module.Index2Param2Write=true
    Module.Index2Param3Write=true
    Write Parameter Index 2=true
   )
   &ScaleColor(
    Scale RGB=&VectorFromFloat(Value=&FloatFromCurve(FloatCurve=DataInterface() {propertyValues {Curve {keys [{interpMode=RCIM_Cubic, tangentMode=RCTM_Break, time=0, value=5}, {interpMode=RCIM_Cubic, tangentMode=RCTM_Break, time=0.2, value=1}]}}}))
    ScaleA=false
    ScaleRGB=true
   ) {enabled=false, module=ScaleColor001}
   &CameraOffset(Camera Offset Amount=&MergeHelper_FloatSwitch(0=2, 1=1, 2=2, 3=1, 4=-5, 5=-5, 6=-5, 7=0, Index=$Particles.PartitionParticles.Partition, Num=::NewEnumerator6)) {module=CameraOffset001}
   /SetVariables(
    Particles.FakeGlowCameraOffset:NiagaraFloat=&MaskFloat(Mask Source=1, Mask Parameter=$Particles.PartitionParticles.Partition, Mask=7, Masked Float=0, Passthrough Float=3)
    Particles.FakeGlowColor:LinearColor=&MultiplyLinearColor(
     Linear Color=<rgba 0.0781 1 1 0.25>
     Scale Factor=&MakeVector4(
      X=1
      Y=1
      Z=1
      W=&FloatFromCurve(FloatCurve=DataInterface() {propertyValues {Curve {keys [
           {arriveTangent=4.6711, interpMode=RCIM_Cubic, leaveTangent=4.6711, tangentMode=RCTM_User, time=0.1, value=0}
           {arriveTangentWeight=0.1067, interpMode=RCIM_Cubic, leaveTangentWeight=0.3202, tangentMode=RCTM_User, time=0.3, value=1}
           {arriveTangentWeight=0.3202, interpMode=RCIM_Cubic, leaveTangentWeight=0.1067, tangentMode=RCTM_User, time=0.7, value=1}
           <at 1 0 RCIM_Cubic>
          ]}}})
     )
    )
    Particles.FakeGlowSize:Vector2f=&Multiply_Vector2D_ByFloat(
     Float=&MergeHelper_FloatSwitch(0=1, 1=1, 2=1, 3=1, 4=1, 5=1, 6=1, 7=2, Index=$Particles.PartitionParticles.Partition, Num=::NewEnumerator6)
     Vector2D=<6 6>
    )
   ) {isSetParameters=true}
   &ScaleMeshSize(Scale Factor=&VectorFromFloat(Value=&MergeHelper_FloatSwitch(0=1, 1=1, 2=1, 3=1, 4=0, 5=0, 6=0, 7=0, Index=$Particles.PartitionParticles.Partition, Num=::NewEnumerator6))) {enabled=false}
   &ScaleColor(
    Scale Alpha=&FloatFromCurve(FloatCurve=DataInterface() {propertyValues {Curve {keys [{interpMode=RCIM_Cubic, tangentMode=RCTM_Break, time=0.7, value=1}, {interpMode=RCIM_Cubic, tangentMode=RCTM_Break, time=1, value=0}]}}})
    ScaleA=true
    ScaleRGB=false
   )
  ]
  rendererClasses [&NiagaraMeshRendererProperties, &NiagaraSpriteRendererProperties]
  renderers [
   {rendererClass=&NiagaraMeshRendererProperties, rendererIndex=0}
   {rendererClass=&NiagaraMeshRendererProperties, rendererIndex=1}
   {rendererClass=&NiagaraMeshRendererProperties, rendererIndex=2}
   {rendererClass=&NiagaraMeshRendererProperties, rendererIndex=3}
   {rendererClass=&NiagaraSpriteRendererProperties, rendererIndex=4}
   {rendererClass=&NiagaraMeshRendererProperties, rendererIndex=5}
  ]
 }
}
summary {
 systemName=NS_Backpack_OboeThorn
 emitters [
  {emitterName=FlamesAndGlows, rendererClasses [&NiagaraSpriteRendererProperties, &NiagaraMeshRendererProperties]}
  {
   emitterName=ReactivityGlows
   rendererClasses [
    &NiagaraSpriteRendererProperties
    &NiagaraLightRendererProperties
    &NiagaraMeshRendererProperties
   ]
  }
  {emitterName=GhostBoosted, rendererClasses [&NiagaraMeshRendererProperties, &NiagaraSpriteRendererProperties]}
  {emitterName=GhostBoosted_InsideHelmet, rendererClasses [&NiagaraMeshRendererProperties]}
  {emitterName=PonyTailGlow, rendererClasses [&NiagaraMeshRendererProperties]}
  {
   emitterName=Trails_Combined
   rendererClasses [&NiagaraMeshRendererProperties, &NiagaraSpriteRendererProperties]
  }
  {
   emitterName=GhostSmall001
   rendererClasses [
    &NiagaraMeshRendererProperties
    &NiagaraSpriteRendererProperties
    &NiagaraLightRendererProperties
   ]
  }
  {
   bEnabled=false
   emitterName=GhostSmall
   rendererClasses [
    &NiagaraMeshRendererProperties
    &NiagaraSpriteRendererProperties
    &NiagaraLightRendererProperties
   ]
  }
 ]
 userVariables [
  User.bIsFrontEnd {type=NiagaraBool}
  User.bIsFrontEndPreview {type=NiagaraBool}
  User.bIsWeaponOn {type=NiagaraBool}
  User.GameTime {type=NiagaraFloat, defaultValue=-2, description=""}
  User.IsSkydiving {type=NiagaraBool}
  User.ReactivityDebug {type=NiagaraBool}
  User.ReactivityDuration {type=NiagaraFloat, defaultValue=4, description=""}
 ]
}
