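# Animation Blueprint for the OboeThorn backpack cosmetic, adding procedural motion to its rig.
# The AnimGraph offsets the pack bone, additively pitches the mask, then chains spring-based Anim Dynamics through four ponytail bones, lower arms and hands.
# The EventGraph sets gravity, front-end simulation and alpha variables, reactivity from the backpack's part modifier, and pickaxe/skydiving state.
# An elimination value raises briefly on enemy-elimination input, or in preview/front end when reactivity is enabled.
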AnimBlueprint /BRCosmetics/Accessories/FORT_Backpacks/Backpack_M_MED_OboeThorn/Meshes/M_MED_OboeThorn_Pack_AnimBP
target_skeleton=&/M_MED_OboeThorn_Pack_Skeleton
graphs [
 AnimGraph {nodes [
   n1 = "Transform (Modify) Bone - Bone: pack"(ComponentPose=Local To Component(LocalPose=Input Pose()), Translation="0,-2.5,0", Rotation="0,0,0") {
    translationMode=BMM_Additive
    translationSpace=BCS_BoneSpace
    alphaBoolBlend {blendOption=Linear}
    boneToModify {boneName=pack}
    translation {y=-2.5}
   }
   n2 = "Transform (Modify) Bone - Bone: mask_01_dyn"(ComponentPose=@n1, Alpha=1, Translation="0,0,0", Rotation="45,0,0", Scale="1,1,1") {
    rotationMode=BMM_Additive
    rotationSpace=BCS_BoneSpace
    translationMode=BMM_Additive
    translationSpace=BCS_BoneSpace
    alphaBoolBlend {blendOption=Linear}
    alphaScaleBiasClamp {bInterpResult=true, interpSpeedDecreasing=5, interpSpeedIncreasing=5}
    boneToModify {boneName=mask_01_dyn}
    rotation {pitch=45}
   }
   n3 = Anim Dynamics - Bone: ponyTail_01_dyn!(ComponentPose=@n2, Alpha=1, GravityOverride:Vector="0,0,-980", ExternalForce="0,0,0") -> (Pose) {
    angularDampingOverride=1
    angularSpringConstant=20000
    bAngularSpring=true
    bOverrideAngularDamping=true
    bUseGravityOverride=true
    gravityScale=0.35
    simulationSpace=RootRelative
    alphaBoolBlend {blendOption=Linear}
    boundBone {boneName=ponyTail_01_dyn}
    componentAppliedLinearAccClamp=<10000 10000>
    componentLinearAccScale=<1 1 1>
    componentLinearVelScale=<1 1 1>
    gravityOverride {z=-980}
    physicsBodyDefinitions [{
      boundBone {boneName=ponyTail_01_dyn}
      boxExtents=<20 35 35>
      constraintSetup {angularLimitsMax=<15 0 5>, angularLimitsMin=<-15 -40 -5>, angularTarget=<1 0 0>}
      localJointOffset=<4 0 0>
     }]
   }
   n4 = Anim Dynamics - Bone: ponyTail_02_dyn!(ComponentPose=@n3, Alpha=1, GravityOverride:Vector="0,0,-980", ExternalForce="0,0,0") -> (Pose) {
    angularDampingOverride=1
    angularSpringConstant=13500
    bAngularSpring=true
    bOverrideAngularDamping=true
    bUseGravityOverride=true
    gravityScale=0.35
    simulationSpace=RootRelative
    alphaBoolBlend {blendOption=Linear}
    boundBone {boneName=ponyTail_02_dyn}
    componentAppliedLinearAccClamp=<10000 10000>
    componentLinearAccScale=<1 1 1>
    componentLinearVelScale=<1 1 1>
    gravityOverride {z=-980}
    physicsBodyDefinitions [{
      boundBone {boneName=ponyTail_02_dyn}
      boxExtents=<20 40 40>
      constraintSetup {angularLimitsMax=<25 15 45>, angularLimitsMin=<-25 -60 -45>, angularTarget=<1 0 0>}
      localJointOffset=<4 0 0>
     }]
   }
   n5 = Anim Dynamics - Bone: ponyTail_03_dyn!(ComponentPose=@n4, Alpha=1, GravityOverride:Vector="0,0,-980", ExternalForce="0,0,0") -> (Pose) {
    angularDampingOverride=1
    angularSpringConstant=11000
    bAngularSpring=true
    bOverrideAngularDamping=true
    bUseGravityOverride=true
    gravityScale=0.35
    simulationSpace=RootRelative
    alphaBoolBlend {blendOption=Linear}
    boundBone {boneName=ponyTail_03_dyn}
    componentAppliedLinearAccClamp=<10000 10000>
    componentLinearAccScale=<1 1 1>
    componentLinearVelScale=<1 1 1>
    gravityOverride {z=-980}
    physicsBodyDefinitions [{
      boundBone {boneName=ponyTail_03_dyn}
      boxExtents=<20 35 40>
      constraintSetup {angularLimitsMax=<25 0 45>, angularLimitsMin=<-25 -55 -45>, angularTarget=<1 0 0>}
      localJointOffset=<4 0 0>
     }]
   }
   n6 = Anim Dynamics - Bone: ponyTail_04_dyn!(ComponentPose=@n5, Alpha=1, GravityOverride:Vector="0,0,-980", ExternalForce="0,0,0") -> (Pose) {
    angularDampingOverride=1
    angularSpringConstant=8000
    bAngularSpring=true
    bOverrideAngularDamping=true
    bUseGravityOverride=true
    gravityScale=0.35
    simulationSpace=RootRelative
    alphaBoolBlend {blendOption=Linear}
    boundBone {boneName=ponyTail_04_dyn}
    componentAppliedLinearAccClamp=<10000 10000>
    componentLinearAccScale=<1 1 1>
    componentLinearVelScale=<1 1 1>
    gravityOverride {z=-980}
    physicsBodyDefinitions [{
      boundBone {boneName=ponyTail_04_dyn}
      boxExtents=<20 40 40>
      constraintSetup {angularLimitsMax=<20 0 45>, angularLimitsMin=<-20 -40 -45>, angularTarget=<1 0 0>}
      localJointOffset=<4 0 0>
     }]
   }
   n7 = Anim Dynamics - Bone: lowerArm_r!(ComponentPose=@n6, Alpha=1, GravityOverride:Vector="0,0,-980", ExternalForce="0,0,0") -> (Pose) {
    angularDampingOverride=1
    angularSpringConstant=40000
    bAngularSpring=true
    bOverrideAngularDamping=true
    bUseGravityOverride=true
    gravityScale=0.35
    linearSpringConstant=10000
    simulationSpace=RootRelative
    alphaBoolBlend {blendOption=Linear}
    boundBone {boneName=lowerArm_r}
    componentAppliedLinearAccClamp=<10000 10000 10000>
    componentLinearAccScale=<1 1 1>
    componentLinearVelScale=<1 1 1>
    gravityOverride {z=-980}
    physicsBodyDefinitions [{
      boundBone {boneName=lowerArm_r}
      boxExtents=<45 45 45>
      constraintSetup {angularLimitsMax=<20 20 20>, angularLimitsMin=<-20 -20 -20>, angularTarget=<1 0 0>}
      localJointOffset=<4 0 0>
     }]
   }
   n8 = Anim Dynamics - Bone: lowerArm_l!(ComponentPose=@n7, Alpha=1, GravityOverride:Vector="0,0,-980", ExternalForce="0,0,0") -> (Pose) {
    angularDampingOverride=1
    angularSpringConstant=40000
    bAngularSpring=true
    bOverrideAngularDamping=true
    bUseGravityOverride=true
    gravityScale=0.35
    linearSpringConstant=10000
    simulationSpace=RootRelative
    alphaBoolBlend {blendOption=Linear}
    boundBone {boneName=lowerArm_l}
    componentAppliedLinearAccClamp=<10000 10000 10000>
    componentLinearAccScale=<1 1 1>
    componentLinearVelScale=<1 1 1>
    gravityOverride {z=-980}
    physicsBodyDefinitions [{
      boundBone {boneName=lowerArm_l}
      boxExtents=<45 45 45>
      constraintSetup {angularLimitsMax=<20 20 20>, angularLimitsMin=<-20 -20 -20>, angularTarget=<1 0 0>}
      localJointOffset=<4 0 0>
     }]
   }
   n9 = Anim Dynamics - Bone: Hand_l!(ComponentPose=@n8, Alpha=1, GravityOverride:Vector="0,0,-980", ExternalForce="0,0,0") -> (Pose) {
    angularDampingOverride=1
    angularSpringConstant=6000
    bAngularSpring=true
    bOverrideAngularDamping=true
    bUseGravityOverride=true
    gravityScale=0.35
    simulationSpace=RootRelative
    alphaBoolBlend {blendOption=Linear}
    boundBone {boneName=Hand_l}
    componentAppliedLinearAccClamp=<10000 10000 10000>
    componentLinearAccScale=<1 1 1>
    componentLinearVelScale=<1 1 1>
    gravityOverride {z=-980}
    physicsBodyDefinitions [{
      boundBone {boneName=Hand_l}
      boxExtents=<30 40 30>
      constraintSetup {angularLimitsMax=<15 15 15>, angularLimitsMin=<-15 -15 -15>, angularTarget=<1 0 0>}
      localJointOffset=<4 0 0>
     }]
   }
   n10 = Anim Dynamics - Bone: Hand_r!(ComponentPose=@n9, Alpha=1, GravityOverride:Vector="0,0,-980", ExternalForce="0,0,0") -> (Pose) {
    angularDampingOverride=1
    angularSpringConstant=6000
    bAngularSpring=true
    bOverrideAngularDamping=true
    bUseGravityOverride=true
    gravityScale=0.35
    simulationSpace=RootRelative
    alphaBoolBlend {blendOption=Linear}
    boundBone {boneName=Hand_r}
    componentAppliedLinearAccClamp=<10000 10000 10000>
    componentLinearAccScale=<1 1 1>
    componentLinearVelScale=<1 1 1>
    gravityOverride {z=-980}
    physicsBodyDefinitions [{
      boundBone {boneName=Hand_r}
      boxExtents=<30 40 30>
      constraintSetup {angularLimitsMax=<15 15 15>, angularLimitsMin=<-15 -15 -15>, angularTarget=<1 0 0>}
      localJointOffset=<4 0 0>
     }]
   }
   Output Pose(Result=Component To Local(ComponentPose=@n10)) {layerGroup=None}
  ]}
 EventGraph {nodes [
   Event BlueprintUpdateAnimation()
   > n1 = Sequence!(execute) -> (then_0, then_1, then_2:Exec)
   > Set Override World Gravity(Override World Gravity:Vector="0,0,-980") -> (Output_Get:Vector)
   > n2 = Branch(Condition=Get bIsInFrontEnd())
   > Set SimlationSpace(SimlationSpace:AnimPhysSimSpaceType Enum=BoneRelative) -> (Output_Get:AnimPhysSimSpaceType Enum)
   > Set AlphaFE(AlphaFE:double=0.4) -> (Output_Get:double)
   Set SimlationSpace(execute=@n2.else)
   > Set AlphaFE(AlphaFE=1)
   n3 = Cast To FortPlayerPawn(execute=@n1.then_2, Object=TryGetFortPawn())
   > n4 = Cast To B_Athena_PartModifier_Backpack_OboeThorn!(
    execute
    Object=GetModifierForPartType!(self:Fort Player Pawn Object Reference=@n3, PartType:EFortCustomPartType Enum=Backpack) -> (ReturnValue:Custom Character Part Modifier Object Reference)
   ) -> (then, CastFailed, AsB Athena Part Modifier Backpack Oboe Thorn:B Athena Part Modifier Backpack Oboe Thorn Object Reference) {is=title K2Node_DynamicCast}
   > Set Reactivity Enabled(Reactivity Enabled:bool=Get Reactivity Enabled!(self:B Athena Part Modifier Backpack Oboe Thorn Object Reference=@n4) -> (Reactivity Enabled:bool)) -> (Output_Get:bool)
   > Branch(Condition=OR Boolean(
     A=IsCosmeticObjectDisplayedAsPreview!(CosmeticObject:Object Reference=@n4) -> (ReturnValue:bool)
     B=IsInFrontEnd!(self:Custom Cosmetic Modifier Base Object Reference=@n4) -> (ReturnValue:bool)
    ))
   > n5 = Branch(Condition=Get Reactivity Enabled!() -> (Reactivity Enabled))
   > Do Once(Reset=@n6.then)
   > Set bEnemyPlayerEliminated(bEnemyPlayerEliminated:double=1) -> (Output_Get:double)
   n6 = Set bEnemyPlayerEliminated(execute=@n5.else, bEnemyPlayerEliminated=0)
   n7 = Event BlueprintProcessAnimInputEvent()
   > Branch(Condition=IsNamed(self=@n7.AnimInputEvent, InEventName=aie.EnemyPlayerEliminated))
   > Set bEnemyPlayerEliminated(bEnemyPlayerEliminated=1)
   > Delay(Duration=3.5)
   > Set bEnemyPlayerEliminated(bEnemyPlayerEliminated=0)
   n8 = Event OnWeaponChanged()
   n9 = Sequence(execute=[@n8.then, @n15.then])
   n10 = PickaxeItemDefinitionHasAnyMetaTag(Tags=Make Array("[0]"=(TagName="Cosmetics.MetaTags.TikiTorch")), bRequiresEquipped=true)
   > n11 = Branch(Condition=@n10)
   > n12 = Branch(Condition=Get bIsSkyDiving())
   Set is_PickaxeOut(execute=[@n14.then, @n13.else, @n12.then, @n11.else], is_PickaxeOut:bool=true) -> (Output_Get:bool)
   Set is_PickaxeOut(execute=[@n13.then, @n12.else], is_PickaxeOut=false)
   n13 = Branch(execute=[@n16.then, @n9.then_1], Condition=@n10)
   n14 = Event OnBeginSkydiving()
   n15 = Event OnCharacterCustomizationLoaded()
   n16 = Event OnEndSkydiving()
   TryGetPawnOwner()
  ]}
]
variables [
 AlphaFE {type=number}
 bEnemyPlayerEliminated {type=number}
 Is Pickaxe_ON {type=number}
 is_PickaxeOut {type=boolean}
 is_PickaxeOut_False {type=boolean}
 OnZipline {type=boolean}
 Override World Gravity {type=Vector}
 Reactivity Enabled {type=boolean}
 Selected_Pistol {type=number}
 SimlationSpace {type=AnimPhysSimSpaceType}
 ZiplineAlpha {type=number}
]
